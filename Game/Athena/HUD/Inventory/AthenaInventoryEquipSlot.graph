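# Athena inventory equip slot widget, built on AthenaInventoryQuickBarSlotButton, showing one quickbar item.
# Slot changes update durability, gadget fuel gauge, empty state and a width scaled by slots the item takes.
# Drag and drop either swaps through AthenaInventoryContext or drops normal or Creative items onto the quickbar, accepting drag-over only for equippable types.
# Hover and selection drive flag-tracked animations, while a ChangeEquipment gameplay-tag check gates the swap indicator.
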WidgetBlueprint /Game/Athena/HUD/Inventory/AthenaInventoryEquipSlot
animations [
 EmptySlotHover {
  binding_count=1
  length_seconds=0.1
  object=&self:EmptySlotHover
  sequence {bindings [EmptyImage {
     id=#1
     possessed_class=Image
     tracks [{
       type=MovieSceneWidgetMaterialTrack
       sections [{type=MovieSceneParameterSection, channels [Animate_Alpha {type=MovieSceneScriptingFloatChannel, keys [[0,0],[2,1]]}]}]
      }]
    }]}
 }
 Scale {
  binding_count=1
  length_seconds=0.1
  object=&self:Scale
  sequence {bindings [OverlayContent {
     id=#2
     possessed_class=Overlay
     tracks [{
       type=MovieScene2DTransformTrack
       sections [{
         type=MovieScene2DTransformSection
         channels [
          Translation.X {type=MovieSceneScriptingFloatChannel}
          Translation.Y {type=MovieSceneScriptingFloatChannel}
          Angle {type=MovieSceneScriptingFloatChannel}
          Scale.X {type=MovieSceneScriptingFloatChannel, keys [[0,1],[2,1.05]]}
          Scale.Y {type=MovieSceneScriptingFloatChannel, keys [[0,1],[2,1.05]]}
          Shear.X {type=MovieSceneScriptingFloatChannel}
          Shear.Y {type=MovieSceneScriptingFloatChannel}
         ]
        }]
      }]
    }]}
 }
 UnderlineSelect {
  binding_count=1
  length_seconds=0.2
  object=&self:UnderlineSelect
  sequence {bindings [Underline {
     id=#3
     possessed_class=Image
     tracks [{
       type=MovieScene2DTransformTrack
       sections [{
         type=MovieScene2DTransformSection
         channels [
          Translation.X {type=MovieSceneScriptingFloatChannel}
          Translation.Y {type=MovieSceneScriptingFloatChannel}
          Angle {type=MovieSceneScriptingFloatChannel}
          Scale.X {type=MovieSceneScriptingFloatChannel, keys [[0,0],[4,1]]}
          Scale.Y {type=MovieSceneScriptingFloatChannel}
          Shear.X {type=MovieSceneScriptingFloatChannel}
          Shear.Y {type=MovieSceneScriptingFloatChannel}
         ]
        }]
      }]
    }]}
 }
]
graphs [
 CanEquipFortItem {nodes [
   n1 = CanEquipFortItem!() -> (then, Object:Object Reference) {is=title K2Node_FunctionEntry}
   > n2 = Cast To FortItem!(execute, Object=@n1) -> (then, CastFailed, AsFort Item:Fort Item Object Reference) {is=title K2Node_DynamicCast}
   n3 = GetType!(self:Fort Item Object Reference=@n2) -> (ReturnValue:EFortItemType Enum=WorldItem)
   > Return Node!(
    execute
    ReturnValue:bool=OR Boolean!(
     A="Equal (Enum)"(A:EFortItemType Enum=@n3, B:EFortItemType Enum=WeaponMelee)
     B="Equal (Enum)"(A=@n3, B=WeaponRanged)
     C:bool="Equal (Enum)"(A=@n3, B=Consumable)
     D:bool="Equal (Enum)"(A=@n3, B=AthenaGadget)
    ) -> (ReturnValue)
   )
   Return Node(execute=@n2.CastFailed, ReturnValue=false)
  ]}
 CanSwap {nodes [
   CanSwap() {is=title K2Node_FunctionEntry}
   > n1 = Cast To FortPlayerController(Object=GetOwningPlayer())
   > Return Node!(
    execute
    CanSwap:bool=CanPerformNativeAction!(
     self:Fort Player Controller Object Reference=@n1
     ActionTag:"Gameplay Tag Structure (by ref)"=MakeLiteralGameplayTag!(Value:Gameplay Tag Structure=(TagName="Gameplay.Action.Player.ChangeEquipment")) -> (ReturnValue:Gameplay Tag Structure)
    ) -> (ReturnValue:bool)
   )
   Return Node(execute=@n1.CastFailed, CanSwap=false)
  ]}
 DropItemOnQuickBar {nodes [
   n1 = DropItemOnQuickBar!() -> (then, Item:Fort Item Object Reference, bIsFromCreativeInventory:bool=~) {is=title K2Node_FunctionEntry}
   > n2 = CanEquipFortItem!(execute, self, Object:Object Reference=@n1) -> (then, ReturnValue:bool) {is=title K2Node_CallFunction}
   > n3 = Branch(Condition=AND Boolean(A=@n2, B=@n1.bIsFromCreativeInventory))
   n4 = Get FortInventoryContext!() -> (ReturnValue:Fort Inventory Context Object Reference) {is=title K2Node_GetBlueprintContextEx}
   n5 = Get Item() -> (Item:Fort Item Object Reference)
   n6 = Get SlotIndex() -> (SlotIndex:int=0)
   n7 = Get QuickBarIndex() -> (QuickBarIndex:EFortQuickBars Enum=Primary)
   > DropCreativeItemOnQuickBar!(
    execute
    self:Fort Inventory Context Object Reference=@n4
    Item:Fort Item Object Reference=@n5
    TargetQuickbar:EFortQuickBars Enum=@n7
    TargetSlot:int=@n6
   ) -> (then, ReturnValue:bool)
   DropItemOnQuickBar!(
    execute=@n3.else
    self:Fort Inventory Context Object Reference=@n4
    Item:Fort Item Object Reference=@n5
    TargetQuickbar:EFortQuickBars Enum=@n7
    TargetSlot:int=@n6
   ) -> (then, ReturnValue:bool) {is=title K2Node_CallFunction}
  ]}
 EventGraph {nodes [
   n1 = Event OnSlottedItemChanged!() -> (OutputDelegate, then, SlottedItem:Fort Item Object Reference) {is=title K2Node_Event}
   > SetItemToRepresent!(
    execute
    self:Athena Item Element Widget Base Object Reference=Get QuickbarSlotDurability() -> (QuickbarSlotDurability:Quickbar Slot Durability Object Reference)
    ItemToRepresent:Fort Item Object Reference=@n1.SlottedItem
   ) -> (then)
   > Update Athena Gadget Fuel Widget!(execute, self, ItemInSlot:Fort Item Object Reference=@n1.SlottedItem) -> (then)
   > UpdateSlotSize!(execute, self, Item:Fort Item Object Reference=@n1.SlottedItem) -> (then)
   n2 = IsValid(Object=@n1.SlottedItem)
   > SetVisibility(
    self=Get ItemIcon() -> (ItemIcon:Athena Item Icon Object Reference)
    InVisibility=Select(Option 0=Collapsed, Option 1=SelfHitTestInvisible, Index=@n2)
   )
   > SetVisibility(
    self=Get EmptyImage() -> (EmptyImage:Image Object Reference)
    InVisibility=Select(Option 0=SelfHitTestInvisible, Option 1=Collapsed, Index=@n2)
   )
   > Set bHasValidItem(bHasValidItem:bool=@n2) -> (Output_Get:bool)
   > n3 = Sequence!(execute) -> (then_0, then_1, then_2:Exec)
   > Branch(Condition=Get bIsSwapTarget() -> (bIsSwapTarget:bool))
   > Set bIsSwapTarget(bIsSwapTarget:bool=false) -> (Output_Get:bool)
   Branch(execute=@n3.then_1, Condition=Get bIsSwapOrigin() -> (bIsSwapOrigin:bool))
   > Set bIsSwapOrigin(bIsSwapOrigin:bool=false) -> (Output_Get:bool)
   Branch(execute=@n3.then_2, Condition=Get bIsSelected() -> (bIsSelected:bool))
   > ResetAnimToStart!(
    execute
    self
    IsAnimStateActive:"Boolean (by ref)"=Get IsUnderlineSelectStateActive() -> (IsUnderlineSelectStateActive:bool)
    Animation:Widget Animation Object Reference=Get UnderlineSelect() -> (UnderlineSelect:Widget Animation Object Reference)
   ) -> (then)
   > ShowSelectedState()
   > ShowHoveredState()
   Event On Hovered()
   > ShowHoveredState()
   > PlaySound2D(Sound=/Game/Sounds/Fort_UI_Sounds/Gameplay/Inventory/UI_Inventory_EquipSlot_Hovered_Cue)
   Event On Released() {is=title K2Node_Event}
   > Parent: On Released() {is=title K2Node_CallParentFunction}
   > Call On Slot Released!(execute:Exec, self:Athena Inventory Equip Slot Object Reference, SlotIndex:int=Get SlotIndex() -> (SlotIndex=0)) -> (then:Exec) {is=title K2Node_CallDelegate}
   Event OnSwapStarted() {is=title K2Node_Event}
   > Branch(Condition=CanSwap!() -> (CanSwap:bool) {is=title K2Node_CallFunction})
   > SetVisibility(self=Get SwapTargetIndicator() -> (SwapTargetIndicator:Overlay Object Reference), InVisibility=SelfHitTestInvisible)
   > Set bIsSwapOrigin(bIsSwapOrigin=true)
   > PlaySound2D(Sound=/Game/Sounds/Fort_UI_Sounds/Gameplay/Inventory/UI_Inventory_EquipSlot_Swap_Drag_Start_Cue)
   Event OnSwapComplete() {is=title K2Node_Event}
   > Branch(Condition=CanSwap())
   > SetVisibility(self=Get SwapTargetIndicator(), InVisibility=Collapsed)
   > SetVisibility(self=Get ScaleBox_InputAction() -> (ScaleBox_InputAction:Scale Box Object Reference), InVisibility=Collapsed)
   > UpdateSlotStateAfterSwap()
   Event OnOwnerDeactivated() {is=title K2Node_Event}
   > SetVisibility(self=Get ScaleBox_InputAction(), InVisibility=Collapsed)
   > ResetAnims()
   > ResetAnims(self:Athena Item Icon Loadout Object Reference=Get AthenaItemIconLoadout() -> (AthenaItemIconLoadout:Athena Item Icon Loadout Object Reference))
   > SetVisibility(self=Get SwapTargetIndicator(), InVisibility=Collapsed)
   n4 = Event OnExtensionWidgetAdded!() -> (OutputDelegate, then, ExtensionSlot:EUIExtensionSlot Enum=Invalid, ExtensionWidget:User Widget Object Reference) {is=title K2Node_Event}
   > Branch(Condition="Equal (Enum)"(A:EUIExtensionSlot Enum=@n4.ExtensionSlot, B:EUIExtensionSlot Enum=InventoryEquipSlot))
   > n5 = AddChildToOverlay!(
    execute
    self:Overlay Object Reference=Get OverlayExtensionWidgets() -> (OverlayExtensionWidgets:Overlay Object Reference)
    Content:Widget Object Reference=@n4.ExtensionWidget
   ) -> (then, ReturnValue:Overlay Slot Object Reference)
   > SetHorizontalAlignment(self=@n5, InHorizontalAlignment=HAlign_Fill)
   > SetVerticalAlignment(self=@n5, InVerticalAlignment=VAlign_Fill)
   Event On Unhovered()
   > n6 = Branch(Condition=Get bHasValidItem() -> (bHasValidItem:bool))
   > SetHoveredState!(execute, self:Athena Item Icon Loadout Object Reference=Get AthenaItemIconLoadout(), bIsHovered:bool=false) -> (then)
   > PlayAnimationReverse(InAnimation=Get Scale() -> (Scale:Widget Animation Object Reference))
   > Set IsScaleStateActive(IsScaleStateActive:bool=false) -> (Output_Get:bool)
   PlayAnimationReverse(execute=@n6.else, InAnimation=Get EmptySlotHover() -> (EmptySlotHover:Widget Animation Object Reference))
   > Set IsEmptySlotHoverStateActive(IsEmptySlotHoverStateActive:bool=false) -> (Output_Get:bool)
   n7 = Event OnDragLeave!() -> (OutputDelegate, then, PointerEvent:Pointer Event Structure, Operation:Drag Drop Operation Object Reference) {is=title K2Node_Event}
   > n8 = Cast To AthenaInventoryDragDrop!(execute, Object=@n7.Operation) -> (then, CastFailed, AsAthena Inventory Drag Drop:Athena Inventory Drag Drop Object Reference) {is=title K2Node_DynamicCast}
   > ShowDropIcon!(execute, self:Athena Inventory Drag Drop Object Reference=@n8, Drop Action:AthenaDragDropAction Enum=NewEnumerator0) -> (then)
   > PlaySound2D(Sound=/Game/Sounds/Fort_UI_Sounds/Gameplay/Inventory/UI_Inventory_EquipSlot_Swap_Drag_Stop_Cue)
   Event OnInitialized()
   > n9 = Cast To AthenaItemIconLoadout!(execute, Object=Get ItemIcon()) -> (then, CastFailed, AsAthena Item Icon Loadout:Athena Item Icon Loadout Object Reference) {is=title K2Node_DynamicCast}
   > Set AthenaItemIconLoadout(AthenaItemIconLoadout:Athena Item Icon Loadout Object Reference=@n9) -> (Output_Get:Athena Item Icon Loadout Object Reference)
   > PlaySound2D(Sound=/Game/Sounds/Fort_UI_Sounds/FrontEnd/New_FrontEnd/DefaultSounds/PressedButtons/FN_UI_Default_Button_Pressed_Cue)
   n10 = Event OnSlotButtonDrop!() -> (OutputDelegate, then, DragDropOperation:Drag Drop Operation Object Reference) {is=title K2Node_Event}
   > HandleSlotButtonDrop!(execute, self, Operation:Drag Drop Operation Object Reference=@n10.DragDropOperation) -> (then)
   Event On Deselected()
   > Set bIsSelected(bIsSelected:bool=false) -> (Output_Get:bool)
   > Set bIsSwapTarget(bIsSwapTarget=false)
   n11 = CanSwap()
   n12 = Get bIsSwapOrigin()
   > SetVisibility(self=Get SwapTargetIndicator(), InVisibility=Select(Option 0=Collapsed, Option 1=SelfHitTestInvisible, Index=AND Boolean(A=@n11, B=@n12)))
   > SetSelectedState!(execute, self:Athena Item Icon Loadout Object Reference=Get AthenaItemIconLoadout(), bIsSelected:bool=false) -> (then)
   > ResetAnimToStart(IsAnimStateActive=Get IsUnderlineSelectStateActive(), Animation=Get UnderlineSelect())
   Event On Selected()
   > Set bIsSelected(bIsSelected=true)
   > ShowSelectedState()
   Event On Pressed() {is=title K2Node_Event}
   > Parent: On Pressed() {is=title K2Node_CallParentFunction}
   > Call On Slot Pressed!(execute:Exec, self:Athena Inventory Equip Slot Object Reference, SlotIndex:int=Get SlotIndex() -> (SlotIndex=0)) -> (then:Exec) {is=title K2Node_CallDelegate}
  ]}
 HandleSlotButtonDrop {nodes [
   n1 = HandleSlotButtonDrop!() -> (then, Operation:Drag Drop Operation Object Reference) {is=title K2Node_FunctionEntry}
   > n2 = Cast To AthenaInventoryDragDrop(Object=@n1)
   n3 = Get ItemIcon()
   n4 = Get Payload!(self:Drag Drop Operation Object Reference=@n2) -> (Payload:Object Reference)
   > Branch(Condition="Not Equal (Object)"!(
     A:Object Reference=@n4
     B:Object Reference=GetItemToRepresent!(self:Athena Item Element Widget Base Object Reference=@n3) -> (ReturnValue:Fort Item Object Reference)
    ) -> (ReturnValue:bool))
   > n5 = Cast To FortItem(Object=@n4)
   n6 = Get AthenaInventoryContext!() -> (ReturnValue:Athena Inventory Context Object Reference) {is=title K2Node_GetBlueprintContextEx}
   > n7 = Branch(Condition=IsSwapping!(self:Athena Inventory Context Object Reference=@n6) -> (ReturnValue:bool))
   > StartOrFinishSwap!(execute, self:Athena Inventory Context Object Reference=@n6, SlotIndex:int=Get SlotIndex() -> (SlotIndex=0)) -> (then)
   > Set bIsSwapTarget(bIsSwapTarget=true)
   Branch(execute=@n7.else, Condition=Get bAddToQuickbarOnDrop() -> (bAddToQuickbarOnDrop:bool))
   > DropItemOnQuickBar!(
    execute
    self:Self Object Reference
    Item=@n5
    bIsFromCreativeInventory:bool=Get bIsFromCreativeInventory!(self:Athena Inventory Drag Drop Object Reference=@n2) -> (bIsFromCreativeInventory:bool)
   ) -> (then)
  ]}
 Hide Athena Gadget Fuel Widget {nodes [
   Hide Athena Gadget Fuel Widget() {is=title K2Node_FunctionEntry}
   > SetItem!(
    execute
    self:Athena Gadget Fuel Widget Object Reference=Get JetpackFuelGauge() -> (JetpackFuelGauge:Athena Gadget Fuel Gauge Object Reference)
    Item:Fort World Item Object Reference
   ) -> (then)
   SetVisibility(self=Get JetpackFuelGauge(), InVisibility=Collapsed)
  ]}
 OnDragDetected {nodes [
   OnDragDetected!() -> (then, MyGeometry:Geometry Structure, PointerEvent:"Pointer Event Structure (by ref)") {is=title K2Node_FunctionEntry}
   > n1 = Set Local_ItemToRepresent(Local_ItemToRepresent:Fort Item Object Reference=GetItemToRepresent(self=Get ItemIcon())) -> (Output_Get:Fort Item Object Reference)
   > Is Valid(InputObject=@n1)
   > Set bIsSwapOrigin(bIsSwapOrigin=true)
   > StartOrFinishSwap(self=Get AthenaInventoryContext(), SlotIndex=Get SlotIndex() -> (SlotIndex=0))
   n2 = GetOwningPlayer()
   > n3 = Construct Object from Class!(
    execute
    Class=Get DragDropVisualClass() -> (DragDropVisualClass:Athena Inventory Drag Visual Class Reference)
    self=@n2
    Item:Fort Item Object Reference=Get Local_ItemToRepresent() -> (Local_ItemToRepresent:Fort Item Object Reference)
   ) -> (then, ReturnValue:Athena Inventory Drag Visual Object Reference)
   > SetOwningPlayer!(execute, self:User Widget Object Reference=@n3, LocalPlayerController:Player Controller Object Reference=@n2) -> (then)
   n4 = Get PointerEvent() -> (PointerEvent:"Pointer Event Structure (by ref)")
   > n5 = Construct Object from Class!(
    execute
    Class=Get DragDropObjectClass() -> (DragDropObjectClass:Athena Inventory Drag Drop Class Reference)
    self
    SplitTheStack:bool=Is Shift Down!(Input:"Input Event Structure (by ref)"=GetInputEventFromPointerEvent!(Event:"Pointer Event Structure (by ref)"=@n4) -> (ReturnValue:Input Event Structure)) -> (ReturnValue:bool)
    bIsFromCreativeInventory:bool=False
    InstigatorSlotIndex:int=Get SlotIndex() -> (SlotIndex=0)
    Tag:String=FromQuickbar
    Payload:Object Reference=Get Local_ItemToRepresent()
    DefaultDragVisual:Widget Object Reference=@n3
    Pivot:EDragPivot Enum=CenterCenter
    Offset:Vector 2D Structure=(X=0,Y=0)
   ) -> (then, ReturnValue:Athena Inventory Drag Drop Object Reference)
   > Return Node!(execute, Operation:Drag Drop Operation Object Reference=@n5)
  ]}
 OnDragOver {nodes [
   n1 = OnDragOver!() -> (then, MyGeometry:Geometry Structure, PointerEvent:Pointer Event Structure, Operation:Drag Drop Operation Object Reference) {is=title K2Node_FunctionEntry}
   > n2 = Cast To AthenaInventoryDragDrop(Object=@n1.Operation)
   n3 = Get ItemIcon()
   n4 = Get Payload(self=@n2)
   > n5 = Branch(Condition="Not Equal (Object)"(A=@n4, B=GetItemToRepresent(self=@n3)))
   > n6 = CanEquipFortItem(Object=@n4)
   > n7 = Branch(Condition=@n6)
   > ShowDropIcon(self=@n2, Drop Action=NewEnumerator2)
   > Return Node!(execute, ReturnValue:bool=true)
   Return Node(execute=[@n7.else, @n5.else, @n2.CastFailed], ReturnValue=false)
  ]}
 OnMouseButtonDown_Border {nodes [
   OnMouseButtonDown_Border!() -> (then, MyGeometry:Geometry Structure, MouseEvent:"Pointer Event Structure (by ref)") {is=title K2Node_FunctionEntry}
   n1 = Get CommonInputSubsystem!() -> (ReturnValue:Common Input Object Reference) {is=title K2Node_GetBlueprintContextEx}
   > n2 = Branch(Condition="Equal (Enum)"(A:ECommonInputType Enum=GetCurrentInputType(self=@n1), B:ECommonInputType Enum=Gamepad))
   > Return Node(ReturnValue:Event Reply Structure=Unhandled())
   Return Node(
    execute=@n2.else
    ReturnValue=DetectDrag!(Reply:"Event Reply Structure (by ref)"=Unhandled(), DragKey:Key Structure=LeftMouseButton) -> (ReturnValue:Event Reply Structure)
   )
  ]}
 OnSlotPressed {nodes [OnSlotPressed!() -> (then, SlotIndex:int) {is=title K2Node_FunctionEntry}]}
 OnSlotReleased {nodes [OnSlotReleased!() -> (then, SlotIndex:int) {is=title K2Node_FunctionEntry}]}
 ResetAnimToStart {nodes [
   n1 = ResetAnimToStart!() -> (then, IsAnimStateActive:"Boolean (by ref)", Animation:Widget Animation Object Reference) {is=title K2Node_FunctionEntry}
   > Branch(Condition=OR Boolean!(A=@n1, B=IsAnimationPlaying(InAnimation=@n1.Animation)) -> (ReturnValue))
   > "Set Boolean (by ref)"!(execute:Exec, Target:"Boolean (by ref)"=@n1, Value:bool) -> (then:Exec) {is=title K2Node_VariableSetRef}
   > PlayAnimation(InAnimation=@n1.Animation, StartAtTime=GetEndTime(self=@n1.Animation), PlayMode=Reverse)
  ]}
 ResetAnims {nodes [
   ResetAnims() {is=title K2Node_FunctionEntry}
   > ResetAnimToStart(IsAnimStateActive=Get IsUnderlineSelectStateActive(), Animation=Get UnderlineSelect()) {is=title K2Node_CallFunction}
   > ResetAnimToStart(IsAnimStateActive=Get IsScaleStateActive() -> (IsScaleStateActive:bool), Animation=Get Scale())
   > ResetAnimToStart(IsAnimStateActive=Get IsEmptySlotHoverStateActive() -> (IsEmptySlotHoverStateActive:bool), Animation=Get EmptySlotHover())
  ]}
 SetCreativeItemInSlot {nodes [
   n1 = SetCreativeItemInSlot!() -> (then, ItemDefinition:Item Definition Base Object Reference) {is=title K2Node_FunctionEntry}
   > n2 = Set CachedItemDefinition(CachedItemDefinition:Item Definition Base Object Reference=@n1) -> (Output_Get:Item Definition Base Object Reference)
   n3 = GetSlottedItem!(self:Athena Quick Bar Slot Button Base Object Reference) -> (ReturnValue:Fort Item Object Reference)
   > n4 = Is Valid(InputObject=@n3)
   > n5 = Branch(Condition="Not Equal (Object)"(
     A:Object Reference=GetItemDefinition!(self:Fort Item Object Reference=@n3) -> (ReturnValue:Fort Item Definition Object Reference)
     B:Object Reference=@n2
    ) -> (ReturnValue:bool=~) {is=title K2Node_PromotableOperator})
   n6 = Set NewTemporaryItem(
    execute=[@n5.then, @n4.Is Not Valid]
    NewTemporaryItem:Fort Item Object Reference=CreateTemporaryItemInstance!(
     ItemDefinition:Item Definition Base Object Reference=Get CachedItemDefinition() -> (CachedItemDefinition:Item Definition Base Object Reference)
     Count:int=1
     Level:int=-1
    ) -> (ReturnValue:Fort Item Object Reference)
   ) -> (Output_Get:Fort Item Object Reference)
   > Is Valid(InputObject=@n6)
   > DropItemOnQuickBar(Item=@n6, bIsFromCreativeInventory=true)
  ]}
 Show Athena Gadget Fuel Widget {nodes [
   n1 = Show Athena Gadget Fuel Widget!() -> (then, Item:Fort Item Object Reference) {is=title K2Node_FunctionEntry}
   > n2 = Cast To FortWorldItem!(execute, Object=@n1) -> (then, CastFailed, AsFort World Item:Fort World Item Object Reference) {is=title K2Node_DynamicCast}
   > SetItem(self=Get JetpackFuelGauge(), Item=@n2)
   SetVisibility(self=Get JetpackFuelGauge(), InVisibility=HitTestInvisible)
  ]}
 ShowHoveredState {nodes [
   ShowHoveredState() {is=title K2Node_FunctionEntry}
   > n1 = Branch(Condition=IsUsingInputType!(InputType:ECommonInputType Enum=Touch, PlayerController:Player Controller Object Reference=GetOwningPlayer()) -> (ReturnValue:bool))
   n2 = Branch(execute=@n1.else, Condition=Get bHasValidItem())
   > SetHoveredState(self=Get AthenaItemIconLoadout(), bIsHovered=true)
   > PlayAnimationForward(InAnimation=Get Scale())
   > Set IsScaleStateActive(IsScaleStateActive=true)
   PlayAnimationForward(execute=@n2.else, InAnimation=Get EmptySlotHover())
   > Set IsEmptySlotHoverStateActive(IsEmptySlotHoverStateActive=true)
  ]}
 ShowSelectedState {nodes [
   ShowSelectedState() {is=title K2Node_FunctionEntry}
   > Branch(Condition=Get bCanShowSelectedState() -> (bCanShowSelectedState:bool))
   > n1 = Set bLocal_IsSwapping(bLocal_IsSwapping:bool=IsSwapping(self=Get AthenaInventoryContext())) -> (Output_Get:bool)
   n2 = CanSwap()
   > SetVisibility(self=Get SwapTargetIndicator(), InVisibility=Select(Option 0=Collapsed, Option 1=SelfHitTestInvisible, Index=AND Boolean(A=@n1, B=@n2)))
   n3 = Get bLocal_IsSwapping() -> (bLocal_IsSwapping:bool)
   n4 = Get bHasValidItem()
   > SetVisibility(self=Get ScaleBox_InputAction(), InVisibility=Select(Option 0=Collapsed, Option 1=SelfHitTestInvisible, Index=OR Boolean(A=@n3, B=@n4)))
   > Branch(Condition=@n4)
   > SetSelectedState(self=Get AthenaItemIconLoadout(), bIsSelected=true)
   > PlayAnimation(InAnimation=Get UnderlineSelect())
   > Set IsUnderlineSelectStateActive(IsUnderlineSelectStateActive:bool=true) -> (Output_Get:bool)
   > SpawnSound2D!(
    execute
    Sound:Sound Base Object Reference=Get ClickSound() -> (ClickSound:Sound Base Object Reference)
    VolumeMultiplier:float=1
    PitchMultiplier:float=1
    StartTime:float=0
    ConcurrencySettings:Sound Concurrency Object Reference
    bPersistAcrossLevelTransition:bool=false
    bAutoDestroy:bool=true
   ) -> (then, ReturnValue:Audio Component Object Reference)
  ]}
 Update Athena Gadget Fuel Widget {nodes [
   n1 = Update Athena Gadget Fuel Widget!() -> (then, ItemInSlot:Fort Item Object Reference) {is=title K2Node_FunctionEntry}
   > n2 = Is Valid(InputObject=@n1)
   Hide Athena Gadget Fuel Widget(execute=@n2.Is Not Valid) {is=title K2Node_CallFunction}
   n3 = Cast To FortGadgetItemDefinition!(execute=@n2.Is Valid, Object=GetItemDefinition(self=Get ItemInSlot() -> (ItemInSlot:Fort Item Object Reference))) -> (then, CastFailed, AsFort Gadget Item Definition:Fort Gadget Item Definition Object Reference) {is=title K2Node_DynamicCast}
   n4 = GetItemType!(self:Fort Item Definition Object Reference=@n3) -> (ReturnValue:EFortItemType Enum=WorldItem)
   > n5 = Branch(Condition=AND Boolean(
     A="Equal (Enum)"(A:EFortItemType Enum=@n4, B:EFortItemType Enum=AthenaGadget)
     B=HasTrackedAttributes!(self:Fort Gadget Item Definition Object Reference=@n3) -> (ReturnValue:bool)
    ))
   > Show Athena Gadget Fuel Widget!(execute, self, Item:Fort Item Object Reference=Get ItemInSlot()) -> (then) {is=title K2Node_CallFunction}
   Hide Athena Gadget Fuel Widget(execute=@n5.else)
   Hide Athena Gadget Fuel Widget(execute=@n3.CastFailed)
  ]}
 UpdateSlotSize {nodes [
   n1 = UpdateSlotSize!() -> (then, Item:Fort Item Object Reference) {is=title K2Node_FunctionEntry}
   > n2 = Is Valid(InputObject=@n1)
   > n3 = Set Local_NumSlotsTaken(Local_NumSlotsTaken:int="To Integer (Byte)"!(InByte:Byte=GetNumberOfSlotsTaken!(self:Fort Item Object Reference=@n1) -> (ReturnValue:Byte=0)) -> (ReturnValue:int=0)) -> (Output_Get:int=0)
   n4 = "Max (Integer)"(A=Get Local_NumSlotsTaken() -> (Local_NumSlotsTaken:int=0), B=1)
   SetWidthOverride(
    execute=[@n3.then, @n2.Is Not Valid]
    self=Get SlotSizeBox() -> (SlotSizeBox:Size Box Object Reference)
    InWidthOverride=float * float(A=Get SingleSlotSize() -> (SingleSlotSize:double), B="To Float (Integer)"(InInt=@n4))
   )
  ]}
 UpdateSlotStateAfterSwap {nodes [
   UpdateSlotStateAfterSwap() {is=title K2Node_FunctionEntry}
   n1 = Get bIsSelected()
   n2 = Get bIsSwapTarget()
   > n3 = Branch(Condition=AND Boolean(A=OR Boolean(A=@n1, B=@n2), B=Get bHasValidItem()))
   > Set bIsSwapOrigin(bIsSwapOrigin=false)
   n4 = Get bHasValidItem()
   > SetVisibility(self=Get ScaleBox_InputAction(), InVisibility=Select(Option 0=Collapsed, Option 1=SelfHitTestInvisible, Index=@n4))
   n5 = IsUsingInputType(InputType=Touch, PlayerController=GetOwningPlayer())
   > Branch(Condition=AND Boolean(A=@n4, B=NOT Boolean(A=@n5)))
   > PlayAnimationForward(InAnimation=Get Scale())
   Branch(execute=@n3.else, Condition=Get bIsSwapOrigin())
   > Set bIsSwapOrigin(bIsSwapOrigin=false)
   > ResetAnimToStart(IsAnimStateActive=Get IsScaleStateActive(), Animation=Get Scale())
   > ResetAnimToStart(IsAnimStateActive=Get IsEmptySlotHoverStateActive(), Animation=Get EmptySlotHover())
  ]}
]
variables [
 AthenaItemIconLoadout {type=AthenaItemIconLoadout_C, default=null}
 bAddToQuickbarOnDrop {type=boolean}
 bCanShowSelectedState {type=boolean, default=true}
 bHasValidItem {type=boolean}
 bIsSelected {type=boolean}
 bIsSwapOrigin {type=boolean}
 bIsSwapTarget {type=boolean}
 ClickSound {type=SoundBase, default=/Game/Sounds/Fort_UI_Sounds/FrontEnd/DefaultSounds/PressedButtons/MSS_UI_Button_Pressed_Default}
 DragDropObjectClass {type=Class, default=&/AthenaInventoryDragDrop.AthenaInventoryDragDrop_C}
 DragDropVisualClass {type=Class, default=&/AthenaInventoryDragVisual.AthenaInventoryDragVisual_C}
 IsEmptySlotHoverStateActive {type=boolean}
 IsScaleStateActive {type=boolean}
 IsUnderlineSelectStateActive {type=boolean}
 OnSlotPressed {}
 OnSlotReleased {}
 SingleSlotSize {type=number, default=122}
]
widgets {
 info {
  inheritedWidgetCount=2
  parentClass=/Script/FortniteUI.AthenaInventoryQuickBarSlotButton
  rootWidgetClass=&SizeBox
  widgetCount=18
 }
 widgets [
  SlotSizeBox {
   widgetClassPath=&SizeBox
   bIsVariable=true
   widget=SlotSizeBox
   properties {bOverride_HeightOverride=true, bOverride_WidthOverride=true, heightOverride=122, widthOverride=122}
  }
  Overlay_179 {widgetClassPath=&Overlay, parent=SlotSizeBox, widget=Overlay_179, slot {class=SizeBoxSlot}}
  OverlayContent {
   widgetClassPath=&Overlay
   parent=Overlay_179
   widget=OverlayContent
   slot {class=OverlaySlot, properties {horizontalAlignment=HAlign_Fill, verticalAlignment=VAlign_Fill}}
  }
  EmptyImage {
   widgetClassPath=&Image
   bIsVariable=true
   parent=OverlayContent
   widget=EmptyImage
   properties {
    visibility=SelfHitTestInvisible
    brush {
     bIsDynamicallyLoaded=false
     drawAs=Box
     imageType=NoImage
     mirroring=NoMirror
     resourceName=None
     resourceObject=/Game/Athena/HUD/Inventory/Art/M_UI_Inventory_LoadoutItem_EmptySlot
     tiling=NoTile
     imageSize=<64 64>
     margin=<ltrb 0.5 0.5 0.5 0.5>
     outlineSettings {
      bUseBrushTransparency=false
      roundingType=HalfHeightRadius
      width=0
      color {colorUseRule=UseColor_Specified, specifiedColor=<rgba 0 0 0 0>}
      cornerRadii=<0 0 0 0>
     }
     tintColor {colorUseRule=UseColor_Specified, specifiedColor=<rgba 1 1 1 1>}
     uVRegion {bIsValid=false, max=<0 0>, min=<0 0>}
    }
   }
   slot {class=OverlaySlot, properties {horizontalAlignment=HAlign_Fill, verticalAlignment=VAlign_Fill}}
  }
  Border_153 {
   widgetClassPath=&Border
   parent=OverlayContent
   widget=Border_153
   properties {
    background {
     bIsDynamicallyLoaded=false
     drawAs=NoDrawType
     imageType=NoImage
     mirroring=NoMirror
     resourceName=None
     resourceObject=null
     tiling=NoTile
     imageSize=<32 32>
     margin=<ltrb 0 0 0 0>
     outlineSettings {
      bUseBrushTransparency=false
      roundingType=HalfHeightRadius
      width=0
      color {colorUseRule=UseColor_Specified, specifiedColor=<rgba 0 0 0 0>}
      cornerRadii=<0 0 0 0>
     }
     tintColor {colorUseRule=UseColor_Specified, specifiedColor=<rgba 1 1 1 1>}
     uVRegion {bIsValid=false, max=<0 0>, min=<0 0>}
    }
    padding=<ltrb 0 0 0 0>
   }
   slot {class=OverlaySlot, properties {horizontalAlignment=HAlign_Fill, verticalAlignment=VAlign_Fill}}
  }
  Overlay_0 {widgetClassPath=&Overlay, parent=Border_153, widget=Overlay_0, slot {class=BorderSlot, properties {padding=<ltrb 0 0 0 0>}}}
  ItemIcon {
   widgetClassPath=&/AthenaItemIconLoadout
   bInherited=true
   parent=Overlay_0
   widget=ItemIcon
   slot {class=OverlaySlot, properties {horizontalAlignment=HAlign_Fill}}
  }
  SizeBox_69 {
   widgetClassPath=&SizeBox
   parent=Overlay_0
   widget=SizeBox_69
   properties {bOverride_HeightOverride=true, bOverride_WidthOverride=true, heightOverride=96, widthOverride=96}
   slot {class=OverlaySlot, properties {horizontalAlignment=HAlign_Fill, verticalAlignment=VAlign_Fill}}
  }
  QuickbarSlotDurability {
   widgetClassPath=/Game/UI/InGame/HUD/QuickbarSlotDurability
   bIsVariable=true
   parent=SizeBox_69
   widget=QuickbarSlotDurability
   properties {visibility=SelfHitTestInvisible}
   slot {class=SizeBoxSlot}
  }
  OverlayExtensionWidgets {
   widgetClassPath=&Overlay
   bIsVariable=true
   parent=OverlayContent
   widget=OverlayExtensionWidgets
   slot {class=OverlaySlot, properties {horizontalAlignment=HAlign_Fill, verticalAlignment=VAlign_Fill}}
  }
  JetpackFuelGauge {
   widgetClassPath=/Game/Athena/UI/InGame/Jetpack/AthenaGadgetFuelGauge
   bIsVariable=true
   parent=OverlayContent
   widget=JetpackFuelGauge
   properties {visibility=Collapsed}
   slot {class=OverlaySlot, properties {horizontalAlignment=HAlign_Right, verticalAlignment=VAlign_Fill, padding=<ltrb 0 6 6 36>}}
  }
  SwapTargetIndicator {
   widgetClassPath=&Overlay
   bIsVariable=true
   parent=OverlayContent
   widget=SwapTargetIndicator
   properties {visibility=Collapsed}
   slot {class=OverlaySlot, properties {horizontalAlignment=HAlign_Fill, verticalAlignment=VAlign_Fill}}
  }
  Shadow {
   widgetClassPath=&Image
   parent=SwapTargetIndicator
   widget=Shadow
   properties {
    visibility=SelfHitTestInvisible
    brush {
     bIsDynamicallyLoaded=false
     drawAs=Image
     imageType=NoImage
     mirroring=NoMirror
     resourceName=None
     resourceObject=/Game/Athena/HUD/Inventory/Art/M_UI_Inventory_DropItem_Shadow
     tiling=NoTile
     imageSize=<32 32>
     margin=<ltrb 0 0 0 0>
     outlineSettings {
      bUseBrushTransparency=false
      roundingType=HalfHeightRadius
      width=0
      color {colorUseRule=UseColor_Specified, specifiedColor=<rgba 0 0 0 0>}
      cornerRadii=<0 0 0 0>
     }
     tintColor {colorUseRule=UseColor_Specified, specifiedColor=<rgba 1 1 1 1>}
     uVRegion {bIsValid=false, max=<0 0>, min=<0 0>}
    }
    renderTransform {angle=0, scale=<1.125 1.125>, shear=<0 0>, translation=<0 0>}
   }
   slot {class=OverlaySlot, properties {horizontalAlignment=HAlign_Fill, verticalAlignment=VAlign_Fill}}
  }
  DropTarget {
   widgetClassPath=&Image
   parent=SwapTargetIndicator
   widget=DropTarget
   properties {
    visibility=SelfHitTestInvisible
    brush {
     bIsDynamicallyLoaded=false
     drawAs=Image
     imageType=NoImage
     mirroring=NoMirror
     resourceName=None
     resourceObject=/Game/Athena/HUD/Inventory/Styles/T-Icon-SwapSlot
     tiling=NoTile
     imageSize=<96 96>
     margin=<ltrb 0 0 0 0>
     outlineSettings {
      bUseBrushTransparency=false
      roundingType=HalfHeightRadius
      width=0
      color {colorUseRule=UseColor_Specified, specifiedColor=<rgba 0 0 0 0>}
      cornerRadii=<0 0 0 0>
     }
     tintColor {colorUseRule=UseColor_Specified, specifiedColor=<rgba 1 1 0.995 1>}
     uVRegion {bIsValid=false, max=<0 0>, min=<0 0>}
    }
   }
   slot {class=OverlaySlot, properties {horizontalAlignment=HAlign_Center, verticalAlignment=VAlign_Center}}
  }
  OverlaySlot {
   widgetClassPath=&Overlay
   bIsVariable=true
   parent=OverlayContent
   widget=OverlaySlot
   slot {class=OverlaySlot, properties {horizontalAlignment=HAlign_Fill, verticalAlignment=VAlign_Fill}}
  }
  Underline {
   widgetClassPath=&Image
   parent=Overlay_179
   widget=Underline
   properties {
    visibility=SelfHitTestInvisible
    brush {
     bIsDynamicallyLoaded=false
     drawAs=Image
     imageType=NoImage
     mirroring=NoMirror
     resourceName=None
     resourceObject=null
     tiling=NoTile
     imageSize=<32 2>
     margin=<ltrb 0 0 0 0>
     outlineSettings {
      bUseBrushTransparency=false
      roundingType=HalfHeightRadius
      width=0
      color {colorUseRule=UseColor_Specified, specifiedColor=<rgba 0 0 0 0>}
      cornerRadii=<0 0 0 0>
     }
     tintColor {colorUseRule=UseColor_Specified, specifiedColor=<rgba 1 1 1 1>}
     uVRegion {bIsValid=false, max=<0 0>, min=<0 0>}
    }
    renderTransform {angle=0, scale=<0 1>, shear=<0 0>, translation=<0 18>}
   }
   slot {class=OverlaySlot, properties {horizontalAlignment=HAlign_Fill, verticalAlignment=VAlign_Bottom}}
  }
  ScaleBox_InputAction {
   widgetClassPath=&ScaleBox
   bIsVariable=true
   parent=Overlay_179
   widget=ScaleBox_InputAction
   properties {stretch=UserSpecified, userSpecifiedScale=0.9, renderTransform {angle=0, scale=<1 1>, shear=<0 0>, translation=<0 54>}}
   slot {class=OverlaySlot, properties {horizontalAlignment=HAlign_Center, verticalAlignment=VAlign_Bottom}}
  }
  InputActionWidget {
   widgetClassPath=&CommonActionWidget
   bInherited=true
   parent=ScaleBox_InputAction
   widget=InputActionWidget
   properties {
    visibility=SelfHitTestInvisible
    iconRimBrush {
     bIsDynamicallyLoaded=false
     drawAs=Image
     imageType=NoImage
     mirroring=NoMirror
     resourceName=None
     resourceObject=/Game/Athena/HUD/Inventory/Art/M_UI_HUD_Inventory_SelectionPrompt_Pulse
     tiling=NoTile
     imageSize=<80 80>
     margin=<ltrb 0 0 0 0>
     outlineSettings {
      bUseBrushTransparency=false
      roundingType=HalfHeightRadius
      width=0
      color {colorUseRule=UseColor_Specified, specifiedColor=<rgba 0 0 0 0>}
      cornerRadii=<0 0 0 0>
     }
     tintColor {colorUseRule=UseColor_Specified, specifiedColor=<rgba 1 1 1 1>}
     uVRegion {bIsValid=false, max=<0 0>, min=<0 0>}
    }
    inputActions [{dataTable=&InputActionDataTable, rowName=Equip}]
   }
   slot {class=ScaleBoxSlot}
  }
 ]
}
