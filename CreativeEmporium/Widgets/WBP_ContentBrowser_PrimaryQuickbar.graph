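# Creative content browser widget showing five Athena inventory equip slots in a row as the primary quickbar.
# Construct gathers the slots into an array and binds their release events, which broadcast OnQuickbarSlotActivated unless an inventory swap is in progress.
# SetItemInSlot offsets the slot index into that array and assigns a creative item.
# Owner activation is forwarded to the slots, then collapses any slot occupied by another slot.
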WidgetBlueprint /CreativeEmporium/Widgets/WBP_ContentBrowser_PrimaryQuickbar
graphs [
 EventGraph {nodes [
   n1 = OnSlotReleased!() -> (OutputDelegate, then, SlotIndex:int) {is=title K2Node_CustomEvent}
   > n2 = Branch(Condition=IsSwapping!(self:Athena Inventory Context Object Reference=Get AthenaInventoryContext!() -> (ReturnValue:Athena Inventory Context Object Reference) {is=title K2Node_GetBlueprintContextEx}) -> (ReturnValue:bool))
   Call On Quickbar Slot Activated!(execute:Exec=@n2.else, self:WBP Content Browser Primary Quickbar Object Reference, SlotIndex:int=@n1.SlotIndex) -> (then:Exec) {is=title K2Node_CallDelegate}
   Event Construct()
   > n3 = Set EquipSlots(EquipSlots:Array of Athena Inventory Equip Slot Object References=Make Array!(
     "[0]":Athena Inventory Equip Slot Object Reference=Get EquipSlot1() -> (EquipSlot1:Athena Inventory Equip Slot Object Reference)
     "[1]":Athena Inventory Equip Slot Object Reference=Get EquipSlot2() -> (EquipSlot2:Athena Inventory Equip Slot Object Reference)
     "[2]":Athena Inventory Equip Slot Object Reference=Get EquipSlot3() -> (EquipSlot3:Athena Inventory Equip Slot Object Reference)
     "[3]":Athena Inventory Equip Slot Object Reference=Get EquipSlot4() -> (EquipSlot4:Athena Inventory Equip Slot Object Reference)
     "[4]":Athena Inventory Equip Slot Object Reference=Get EquipSlot5() -> (EquipSlot5:Athena Inventory Equip Slot Object Reference)
    ) -> (Array:Array of Athena Inventory Equip Slot Object References)) -> (Output_Get:Array of Athena Inventory Equip Slot Object References)
   > n4 = For Each Loop(Array:Array of Athena Inventory Equip Slot Object References=@n3) -> (Array Element:Athena Inventory Equip Slot Object Reference)
   > Bind Event to On Slot Released(self:Athena Inventory Equip Slot Object Reference=@n4, Delegate=Create Event()) {is=title K2Node_AddDelegate}
  ]}
 OnOwnerActivated {nodes [
   OnOwnerActivated() {is=title K2Node_FunctionEntry}
   > NativeOnOwnerActivated(self:Athena Quick Bar Slot Button Base Object Reference=Get EquipSlots() -> (EquipSlots:Array of Athena Inventory Equip Slot Object References))
   > UpdateSlotVisibilities()
  ]}
 OnOwnerDeactivated {nodes [
   OnOwnerDeactivated() {is=title K2Node_FunctionEntry}
   > NativeOnOwnerDeactivated(self:Athena Quick Bar Slot Button Base Object Reference=Get EquipSlots())
  ]}
 OnQuickbarSlotActivated {nodes [OnQuickbarSlotActivated!() -> (then, SlotIndex:int) {is=title K2Node_FunctionEntry}]}
 SetItemInSlot {nodes [
   n1 = SetItemInSlot!() -> (then, SlotIndex:int, ItemDefinition:Item Definition Base Object Reference) {is=title K2Node_FunctionEntry}
   n2 = int - int(A=@n1)
   n3 = Get EquipSlots()
   > Branch(Condition=Is Valid Index!(TargetArray:Array of Athena Inventory Equip Slot Object References=@n3, IndexToTest:int=@n2) -> (ReturnValue:bool))
   > SetCreativeItemInSlot!(
    execute
    self:Athena Inventory Equip Slot Object Reference="Get (a copy)"(Array:Array of Athena Inventory Equip Slot Object References=@n3, Dimension 1=@n2) -> (Output:Athena Inventory Equip Slot Object Reference)
    ItemDefinition:Item Definition Base Object Reference=@n1.ItemDefinition
   ) -> (then)
  ]}
 UpdateSlotVisibilities {nodes [
   UpdateSlotVisibilities() {is=title K2Node_FunctionEntry}
   > n1 = For Each Loop(Array=Get EquipSlots())
   n2 = Get FortInventoryContext!() -> (ReturnValue:Fort Inventory Context Object Reference) {is=title K2Node_GetBlueprintContextEx}
   n3 = Get SlotIndex!(self:Athena Quick Bar Slot Button Base Object Reference=@n1) -> (SlotIndex:int=0)
   > SetVisibility(
    self=@n1
    InVisibility=Select(
     Option 0=SelfHitTestInvisible
     Option 1=Collapsed
     Index=IsSlotOccupiedByAnotherSlot!(self:Fort Inventory Context Object Reference=@n2, InQuickBar:EFortQuickBars Enum=Primary, Slot:int=@n3) -> (ReturnValue:bool)
    )
   )
  ]}
]
variables [
 EquipSlots {type="array<AthenaInventoryEquipSlot_C>", default []}
 OnQuickbarSlotActivated {}
]
widgets {
 info {inheritedWidgetCount=0, parentClass=&UserWidget, rootWidgetClass=&HorizontalBox, widgetCount=6}
 widgets [
  HorizontalBox_49 {widgetClassPath=&HorizontalBox, widget=HorizontalBox_49}
  EquipSlot1 {
   widgetClassPath=/Game/Athena/HUD/Inventory/AthenaInventoryEquipSlot
   bIsVariable=true
   parent=HorizontalBox_49
   widget=EquipSlot1
   properties {
    bAddToQuickbarOnDrop=true
    bCanShowSelectedState=false
    bInteractableWhenSelected=false
    bIsFocusable=false
    bSelectable=false
    bShouldSelectUponReceivingFocus=false
    holdData=&CommonUI_HoldData_Base_C
    slotIndex=1
   }
   slot {class=HorizontalBoxSlot, properties {padding=<ltrb 0 0 10 0>}}
  }
  EquipSlot2 {
   widgetClassPath=&AthenaInventoryEquipSlot
   bIsVariable=true
   parent=HorizontalBox_49
   widget=EquipSlot2
   properties {
    bAddToQuickbarOnDrop=true
    bCanShowSelectedState=false
    bIsFocusable=false
    bSelectable=false
    holdData=&CommonUI_HoldData_Base_C
    slotIndex=2
   }
   slot {class=HorizontalBoxSlot, properties {padding=<ltrb 0 0 10 0>}}
  }
  EquipSlot3 {
   widgetClassPath=&AthenaInventoryEquipSlot
   bIsVariable=true
   parent=HorizontalBox_49
   widget=EquipSlot3
   properties {
    bAddToQuickbarOnDrop=true
    bCanShowSelectedState=false
    bIsFocusable=false
    bSelectable=false
    holdData=&CommonUI_HoldData_Base_C
    slotIndex=3
   }
   slot {class=HorizontalBoxSlot, properties {padding=<ltrb 0 0 10 0>}}
  }
  EquipSlot4 {
   widgetClassPath=&AthenaInventoryEquipSlot
   bIsVariable=true
   parent=HorizontalBox_49
   widget=EquipSlot4
   properties {
    bAddToQuickbarOnDrop=true
    bCanShowSelectedState=false
    bIsFocusable=false
    bSelectable=false
    holdData=&CommonUI_HoldData_Base_C
    slotIndex=4
   }
   slot {class=HorizontalBoxSlot, properties {padding=<ltrb 0 0 10 0>}}
  }
  EquipSlot5 {
   widgetClassPath=&AthenaInventoryEquipSlot
   bIsVariable=true
   parent=HorizontalBox_49
   widget=EquipSlot5
   properties {
    bAddToQuickbarOnDrop=true
    bCanShowSelectedState=false
    bIsFocusable=false
    bSelectable=false
    holdData=&CommonUI_HoldData_Base_C
    slotIndex=5
   }
   slot {class=HorizontalBoxSlot}
  }
 ]
}
