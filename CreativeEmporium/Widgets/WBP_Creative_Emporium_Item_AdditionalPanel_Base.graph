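# Base widget for an additional-info panel on Creative Emporium items, derived from FortEmporiumItemInfo.
# A size-constrained overlay layers a material background, a title, a clipped named slot for additional content, and a MORE button hanging past the bottom edge.
# PreConstruct copies the Title variable into the title text.
# On receiving focus, the panel passes focus to the MORE button, which has no click logic in this graph.
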WidgetBlueprint /CreativeEmporium/Widgets/WBP_Creative_Emporium_Item_AdditionalPanel_Base
graphs [
 EventGraph {nodes [
   Event PreConstruct()
   > "SetText (Text)"(self=Get Text_Title() -> (Text_Title:Common Text Object Reference), InText=Get Title() -> (Title:Text))
  ]}
 OnFocusReceived {nodes [
   OnFocusReceived!() -> (then, MyGeometry:Geometry Structure, InFocusEvent:Focus Event Structure) {is=title K2Node_FunctionEntry}
   > SetFocus(self=Get Button_More() -> (Button_More:WBP UIKit Button Loud Object Reference))
   > Return Node!(execute, ReturnValue:Event Reply Structure=Handled())
  ]}
]
variables [Title {type=string, default="[PH] Title"}]
widgets {
 info {inheritedWidgetCount=0, parentClass=/Script/CreativeEmporiumUI.FortEmporiumItemInfo, rootWidgetClass=&SizeBox, widgetCount=7}
 widgets [
  SizeBox_Root {
   widgetClassPath=&SizeBox
   widget=SizeBox_Root
   properties {
    bOverride_MaxDesiredHeight=true
    bOverride_MaxDesiredWidth=true
    bOverride_MinDesiredHeight=true
    bOverride_MinDesiredWidth=true
    maxDesiredHeight=368
    maxDesiredWidth=1032
    minDesiredHeight=368
    minDesiredWidth=696
   }
  }
  Overlay_Content {widgetClassPath=&Overlay, parent=SizeBox_Root, widget=Overlay_Content, slot {class=SizeBoxSlot}}
  Image_Background {
   widgetClassPath=&Image
   parent=Overlay_Content
   widget=Image_Background
   properties {brush {
     bIsDynamicallyLoaded=false
     drawAs=Image
     imageType=NoImage
     mirroring=NoMirror
     resourceName=None
     resourceObject=/CreativeEmporium/Materials/MI_Creative_Emporium_Background_Item_Description
     tiling=NoTile
     imageSize=<32 32>
     margin=<ltrb 0 0 0 0>
     outlineSettings {
      bUseBrushTransparency=false
      roundingType=HalfHeightRadius
      width=0
      color {colorUseRule=UseColor_Specified, specifiedColor=<rgba 0 0 0 0>}
      cornerRadii=<0 0 0 0>
     }
     tintColor {colorUseRule=UseColor_Specified, specifiedColor=<rgba 1 1 1 1>}
     uVRegion {bIsValid=false, max=<0 0>, min=<0 0>}
    }}
   slot {class=OverlaySlot, properties {horizontalAlignment=HAlign_Fill, verticalAlignment=VAlign_Fill}}
  }
  VerticalBox_Content {
   widgetClassPath=&VerticalBox
   parent=Overlay_Content
   widget=VerticalBox_Content
   slot {class=OverlaySlot, properties {horizontalAlignment=HAlign_Fill, padding=<ltrb 48 48 48 48>}}
  }
  Text_Title {
   widgetClassPath=&CommonTextBlock
   bIsVariable=true
   parent=VerticalBox_Content
   widget=Text_Title
   properties {
    style=/FortUILibrary/Styles/Typography/_TextStyle_SubH4._TextStyle_SubH4_C
    text="[PH] Title"
    font {
     bForceMonospaced=false
     bMaterialIsStencil=false
     fontMaterial=null
     fontObject=&BurbankSmall
     letterSpacing=0
     monospacedWidth=1
     size=24
     skewAmount=0
     typefaceFontName=Black
     outlineSettings {
      bApplyOutlineToDropShadows=false
      bMiteredCorners=false
      bSeparateFillAlpha=false
      outlineMaterial=null
      outlineSize=0
      outlineColor=<rgba 0 0 0 1>
     }
    }
    shadowOffset=<0 0>
    strikeBrush {
     bIsDynamicallyLoaded=false
     drawAs=NoDrawType
     imageType=NoImage
     mirroring=NoMirror
     resourceName=None
     resourceObject=null
     tiling=NoTile
     imageSize=<32 32>
     margin=<ltrb 0 0 0 0>
     outlineSettings {
      bUseBrushTransparency=false
      roundingType=HalfHeightRadius
      width=0
      color {colorUseRule=UseColor_Specified, specifiedColor=<rgba 0 0 0 0>}
      cornerRadii=<0 0 0 0>
     }
     tintColor {colorUseRule=UseColor_Specified, specifiedColor=<rgba 1 1 1 1>}
     uVRegion {bIsValid=false, max=<0 0>, min=<0 0>}
    }
   }
   slot {class=VerticalBoxSlot, properties {horizontalAlignment=HAlign_Left, verticalAlignment=VAlign_Top}}
  }
  NamedSlot_AdditionalInformation {
   widgetClassPath=&NamedSlot
   parent=VerticalBox_Content
   widget=NamedSlot_AdditionalInformation
   properties {clipping=ClipToBounds}
   slot {class=VerticalBoxSlot, properties {verticalAlignment=VAlign_Top, padding=<ltrb 0 27 0 0>, size {sizeRule=Fill, value=1}}}
  }
  Button_More {
   widgetClassPath=/FortUILibrary/UIKit/Buttons/WBP_UIKit_Button_Loud
   bIsVariable=true
   parent=Overlay_Content
   widget=Button_More
   properties {buttonSize=Small, text=MORE}
   slot {class=OverlaySlot, properties {horizontalAlignment=HAlign_Center, verticalAlignment=VAlign_Bottom, padding=<ltrb 0 0 0 -24>}}
  }
 ]
}
