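# Description variant of the Creative Emporium item additional panel, a Widget Blueprint derived from WBP_Creative_Emporium_Item_AdditionalPanel_Base.
# It has no root widget of its own and adds one CommonTextBlock, Text_Description, alongside a single named slot.
# The text auto-wraps with clipping and ellipsis overflow in a small white Burbank style, showing placeholder copy.
# The event graph is empty, so this asset contains no logic to set the text.
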WidgetBlueprint /CreativeEmporium/Widgets/WBP_Creative_Emporium_Item_AdditionalPanel_Description
graphs [EventGraph {links [], nodes []}]
variables []
widgets {
 info {
  inheritedWidgetCount=0
  namedSlotCount=1
  parentClass=&/WBP_Creative_Emporium_Item_AdditionalPanel_Base.WBP_Creative_Emporium_Item_AdditionalPanel_Base_C
  rootWidgetClass=null
  widgetCount=1
 }
 widgets [Text_Description {
   widgetClassPath=&CommonTextBlock
   widget=Text_Description
   properties {
    autoWrapText=true
    bIsScrollingEnabled=false
    clipping=ClipToBounds
    style=/FortUILibrary/Styles/Typography/Foundation/TextStyle-BurbankSmall-S-White.TextStyle-BurbankSmall-S-White_C
    text="[PH] Some description"
    textOverflowPolicy=Ellipsis
    font {
     bForceMonospaced=false
     bMaterialIsStencil=false
     fontMaterial=null
     fontObject=&BurbankSmall
     letterSpacing=0
     monospacedWidth=1
     size=17
     skewAmount=0
     typefaceFontName=Black
     outlineSettings {
      bApplyOutlineToDropShadows=false
      bMiteredCorners=false
      bSeparateFillAlpha=false
      outlineMaterial=null
      outlineSize=0
      outlineColor=<rgba 0 0 0 1>
     }
    }
    shadowOffset=<0 0>
   }
  }]
}
