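# Widget Blueprint for one row in a Creative Emporium item details display, showing an item name and a stat, based on CommonUserWidget.
# A vertical box stacks a padded horizontal row above a thin white divider image.
# In the row, the name text sits on the left and the stat text fills the remaining width, aligned right.
# It has no graph logic or variables, and both texts hold placeholder strings.
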WidgetBlueprint /CreativeEmporium/Widgets/WBP_Creative_Emporium_Item_Details_Element
graphs [EventGraph {links [], nodes []}]
variables []
widgets {
 info {inheritedWidgetCount=0, parentClass=&CommonUserWidget, rootWidgetClass=&VerticalBox, widgetCount=5}
 widgets [
  VerticalBox_Root {widgetClassPath=&VerticalBox, widget=VerticalBox_Root}
  HorizontalBox_Content {
   widgetClassPath=&HorizontalBox
   parent=VerticalBox_Root
   widget=HorizontalBox_Content
   slot {class=VerticalBoxSlot, properties {verticalAlignment=VAlign_Top, padding=<ltrb 0 7 0 7>}}
  }
  Text_ItemName {
   widgetClassPath=&CommonTextBlock
   parent=HorizontalBox_Content
   widget=Text_ItemName
   properties {
    style=/FortUILibrary/Styles/Typography/Foundation/TextStyle_BurbankSmall_S_White.TextStyle_BurbankSmall_S_White_C
    text="[PH] Item Name"
    font {
     bForceMonospaced=false
     bMaterialIsStencil=false
     fontMaterial=null
     fontObject=&BurbankSmall
     letterSpacing=0
     monospacedWidth=1
     size=18
     skewAmount=0
     typefaceFontName=Black
     outlineSettings {
      bApplyOutlineToDropShadows=false
      bMiteredCorners=false
      bSeparateFillAlpha=false
      outlineMaterial=null
      outlineSize=0
      outlineColor=<rgba 0 0 0 1>
     }
    }
    shadowOffset=<0 0>
   }
   slot {class=HorizontalBoxSlot, properties {horizontalAlignment=HAlign_Left, verticalAlignment=VAlign_Center}}
  }
  Text_ItemStat {
   widgetClassPath=&CommonTextBlock
   parent=HorizontalBox_Content
   widget=Text_ItemStat
   properties {
    style=&TextStyle_BurbankSmall_S_White_C
    text="[PH] Item Stat"
    font {
     bForceMonospaced=false
     bMaterialIsStencil=false
     fontMaterial=null
     fontObject=&BurbankSmall
     letterSpacing=0
     monospacedWidth=1
     size=18
     skewAmount=0
     typefaceFontName=Black
     outlineSettings {
      bApplyOutlineToDropShadows=false
      bMiteredCorners=false
      bSeparateFillAlpha=false
      outlineMaterial=null
      outlineSize=0
      outlineColor=<rgba 0 0 0 1>
     }
    }
    shadowOffset=<0 0>
   }
   slot {class=HorizontalBoxSlot, properties {horizontalAlignment=HAlign_Right, verticalAlignment=VAlign_Center, size {sizeRule=Fill, value=1}}}
  }
  Image_BottomLine {
   widgetClassPath=&Image
   parent=VerticalBox_Root
   widget=Image_BottomLine
   properties {
    visibility=SelfHitTestInvisible
    brush {
     bIsDynamicallyLoaded=false
     drawAs=Image
     imageType=NoImage
     mirroring=NoMirror
     resourceName=None
     resourceObject=null
     tiling=NoTile
     imageSize=<32 2>
     margin=<ltrb 0 0 0 0>
     outlineSettings {
      bUseBrushTransparency=false
      roundingType=HalfHeightRadius
      width=0
      color {colorUseRule=UseColor_Specified, specifiedColor=<rgba 0 0 0 0>}
      cornerRadii=<0 0 0 0>
     }
     tintColor {colorUseRule=UseColor_Specified, specifiedColor=<rgba 1 1 1 1>}
     uVRegion {bIsValid=false, max=<0 0>, min=<0 0>}
    }
   }
   slot {class=VerticalBoxSlot, properties {verticalAlignment=VAlign_Top}}
  }
 ]
}
