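# Widget Blueprint for the Details variant of the Creative Emporium item additional panel, derived from the shared AdditionalPanel base widget.
# It has no root of its own, only a vertical box that appears to fill the parent's named slot.
# The box stacks four Details Element widgets, top-aligned with equal fill.
# It adds no variables or graph logic, so any behavior comes from the base and element widgets.
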WidgetBlueprint /CreativeEmporium/Widgets/WBP_Creative_Emporium_Item_AdditionalPanel_Details
graphs [EventGraph {links [], nodes []}]
variables []
widgets {
 info {
  inheritedWidgetCount=0
  namedSlotCount=1
  parentClass=&/WBP_Creative_Emporium_Item_AdditionalPanel_Base.WBP_Creative_Emporium_Item_AdditionalPanel_Base_C
  rootWidgetClass=null
  widgetCount=5
 }
 widgets [
  VerticalBox_45 {widgetClassPath=&VerticalBox, widget=VerticalBox_45}
  WBP_Creative_Emporium_Item_Details_Element {
   widgetClassPath=&/WBP_Creative_Emporium_Item_Details_Element
   parent=VerticalBox_45
   widget=WBP_Creative_Emporium_Item_Details_Element
   slot {class=VerticalBoxSlot, properties {verticalAlignment=VAlign_Top, size {sizeRule=Fill, value=1}}}
  }
  WBP_Creative_Emporium_Item_Details_Element_1 {
   widgetClassPath=&WBP_Creative_Emporium_Item_Details_Element
   parent=VerticalBox_45
   widget=WBP_Creative_Emporium_Item_Details_Element_1
   slot {class=VerticalBoxSlot, properties {verticalAlignment=VAlign_Top, size {sizeRule=Fill, value=1}}}
  }
  WBP_Creative_Emporium_Item_Details_Element_2 {
   widgetClassPath=&WBP_Creative_Emporium_Item_Details_Element
   parent=VerticalBox_45
   widget=WBP_Creative_Emporium_Item_Details_Element_2
   slot {class=VerticalBoxSlot, properties {verticalAlignment=VAlign_Top, size {sizeRule=Fill, value=1}}}
  }
  WBP_Creative_Emporium_Item_Details_Element_3 {
   widgetClassPath=&WBP_Creative_Emporium_Item_Details_Element
   parent=VerticalBox_45
   widget=WBP_Creative_Emporium_Item_Details_Element_3
   slot {class=VerticalBoxSlot, properties {verticalAlignment=VAlign_Top, size {sizeRule=Fill, value=1}}}
  }
 ]
}
