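# Widget Blueprint for the additional-information section of an item in the Creative Emporium, built on FortEmporiumItemInfo.
# A horizontal box holds a Description panel on the left and a Details panel that fills the remaining width, aligned top-right.
# The OnFocusReceived override forwards focus to the Description panel and returns Handled.
# It declares no variables of its own, and its event graph is empty.
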WidgetBlueprint /CreativeEmporium/Widgets/WBP_Creative_Emporium_Item_AdditionalInformation
graphs [
 EventGraph {links [], nodes []}
 OnFocusReceived {nodes [
   OnFocusReceived!() -> (then, MyGeometry:Geometry Structure, InFocusEvent:Focus Event Structure) {is=title K2Node_FunctionEntry}
   > SetFocus(self=Get WBP_Creative_Emporium_Item_AdditionalPanel_Description() -> (WBP_Creative_Emporium_Item_AdditionalPanel_Description:WBP Creative Emporium Item Additional Panel Description Object Reference))
   > Return Node!(execute, ReturnValue:Event Reply Structure=Handled())
  ]}
]
variables []
widgets {
 info {
  inheritedWidgetCount=0
  parentClass=/Script/CreativeEmporiumUI.FortEmporiumItemInfo
  rootWidgetClass=&HorizontalBox
  widgetCount=3
 }
 widgets [
  HorizontalBox_Root {widgetClassPath=&HorizontalBox, widget=HorizontalBox_Root}
  WBP_Creative_Emporium_Item_AdditionalPanel_Description {
   widgetClassPath=&/WBP_Creative_Emporium_Item_AdditionalPanel_Description
   bIsVariable=true
   parent=HorizontalBox_Root
   widget=WBP_Creative_Emporium_Item_AdditionalPanel_Description
   slot {class=HorizontalBoxSlot, properties {horizontalAlignment=HAlign_Left, verticalAlignment=VAlign_Top}}
  }
  WBP_Creative_Emporium_Item_AdditionalPanel_Details {
   widgetClassPath=&/WBP_Creative_Emporium_Item_AdditionalPanel_Details
   parent=HorizontalBox_Root
   widget=WBP_Creative_Emporium_Item_AdditionalPanel_Details
   slot {
    class=HorizontalBoxSlot
    properties {horizontalAlignment=HAlign_Right, verticalAlignment=VAlign_Top, padding=<ltrb 72 0 0 0>, size {sizeRule=Fill, value=1}}
   }
  }
 ]
}
