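# Creative Emporium item info panel, built on FortEmporiumItemInfo, that shows a Fort item's name and type in a vertical stack.
# UpdateWithFortItem stores the item, checks validity, clears both text blocks, then fills them from the item's display name and singular type name.
# PreConstruct sets placeholder text at design time.
# The "Created by : {CreatorName}" line is never updated by this widget's graphs.
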WidgetBlueprint /CreativeEmporium/Widgets/WBP_Creative_Emporium_Item_Info
graphs [
 EventGraph {nodes [
   n1 = Event PreConstruct()
   > Branch(Condition=@n1.IsDesignTime)
   > SetDesignerText!(execute, self, InItemName:Text="LOCTEXT(\"[PH] Item Name\")", InItemType:Text="LOCTEXT(\"[PH] Item Type\")") -> (then)
   n2 = Event UpdateWithFortItem!() -> (OutputDelegate, then, Item:Fort Item Object Reference, ParentDefinition:Item Definition Base Object Reference, bAllowInteractiveTags:bool) {is=title K2Node_Event}
   > n3 = Set Item(Item:Fort Item Object Reference=@n2.Item) -> (Output_Get:Fort Item Object Reference)
   > Is Valid(InputObject=@n3)
   > ResetText()
   > UdpateItemName()
   > UpdateItemType()
   Event Construct()
   Event Tick()
  ]}
 ResetText {nodes [
   ResetText() {is=title K2Node_FunctionEntry}
   > "SetText (Text)"(
    self=[Get Text_Name() -> (Text_Name:Common Text Object Reference), Get Text_Type() -> (Text_Type:Common Text Object Reference)]
    InText=GetEmptyText!() -> (ReturnValue:Text)
   )
  ]}
 SetDesignerText {nodes [
   SetDesignerText!() -> (then, InItemName:Text, InItemType:Text) {is=title K2Node_FunctionEntry}
   > "SetText (Text)"(self=Get Text_Name(), InText=Get InItemName() -> (InItemName:Text))
   > "SetText (Text)"(self=Get Text_Type(), InText=Get InItemType() -> (InItemType:Text))
  ]}
 UdpateItemName {nodes [
   UdpateItemName() {is=title K2Node_FunctionEntry}
   > "SetText (Text)"(
    self=Get Text_Name()
    InText=GetDisplayName!(self:Fort Item Object Reference=Get Item() -> (Item:Fort Item Object Reference)) -> (ReturnValue:Text)
   )
  ]}
 UpdateItemType {nodes [
   UpdateItemType() {is=title K2Node_FunctionEntry}
   > "SetText (Text)"(self=Get Text_Type(), InText=GetItemTypeName!(self:Fort Item Object Reference=Get Item(), bUsePlural:bool=false) -> (ReturnValue:Text))
  ]}
]
variables [Item {type=FortItem, default=null}]
widgets {
 info {inheritedWidgetCount=0, parentClass=/Script/CreativeEmporiumUI.FortEmporiumItemInfo, rootWidgetClass=&VerticalBox, widgetCount=4}
 widgets [
  Root {widgetClassPath=&VerticalBox, widget=Root}
  Text_Name {
   widgetClassPath=&CommonTextBlock
   bIsVariable=true
   parent=Root
   widget=Text_Name
   properties {
    mobileFontSizeMultiplier=0.7
    text="[PH] Item Name"
    textTransformPolicy=ToUpper
    font {
     bForceMonospaced=false
     bMaterialIsStencil=false
     fontMaterial=null
     fontObject=&Burbank
     letterSpacing=0
     monospacedWidth=1
     size=72
     skewAmount=0
     typefaceFontName=BurbankRegular-Black
     outlineSettings {
      bApplyOutlineToDropShadows=false
      bMiteredCorners=false
      bSeparateFillAlpha=false
      outlineMaterial=null
      outlineSize=1
      outlineColor=<rgba 0 0 0 0.1>
     }
    }
    shadowOffset=<0 0>
   }
   slot {class=VerticalBoxSlot, properties {horizontalAlignment=HAlign_Left, verticalAlignment=VAlign_Top}}
  }
  Text_Type {
   widgetClassPath=&CommonTextBlock
   bIsVariable=true
   parent=Root
   widget=Text_Type
   properties {
    mobileFontSizeMultiplier=0.7
    text="[PH] Item Type"
    textTransformPolicy=ToUpper
    font {
     bForceMonospaced=false
     bMaterialIsStencil=false
     fontMaterial=null
     fontObject=&BurbankSmall
     letterSpacing=0
     monospacedWidth=1
     size=24
     skewAmount=0
     typefaceFontName=Bold
     outlineSettings {
      bApplyOutlineToDropShadows=false
      bMiteredCorners=false
      bSeparateFillAlpha=false
      outlineMaterial=null
      outlineSize=1
      outlineColor=<rgba 0 0 0 0.1>
     }
    }
    shadowOffset=<0 0>
   }
   slot {class=VerticalBoxSlot, properties {horizontalAlignment=HAlign_Left, verticalAlignment=VAlign_Top}}
  }
  CommonTextBlock_594 {
   widgetClassPath=&CommonTextBlock
   parent=Root
   widget=CommonTextBlock_594
   properties {
    text="Created by : {CreatorName}"
    font {
     bForceMonospaced=false
     bMaterialIsStencil=false
     fontMaterial=null
     fontObject=&BurbankSmall
     letterSpacing=0
     monospacedWidth=1
     size=20
     skewAmount=0
     typefaceFontName=Black
     outlineSettings {
      bApplyOutlineToDropShadows=false
      bMiteredCorners=false
      bSeparateFillAlpha=false
      outlineMaterial=null
      outlineSize=0
      outlineColor=<rgba 0 0 0 1>
     }
    }
    shadowOffset=<0 0>
   }
   slot {class=VerticalBoxSlot, properties {padding=<ltrb 0 28 0 0>}}
  }
 ]
}
