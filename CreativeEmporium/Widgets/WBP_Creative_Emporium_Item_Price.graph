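# Creative Emporium item panel widget showing price and license, with Place Now, Add to Library, License and ellipsis buttons.
# UpdateWithFortItem copies Emporium data from the item definition, then fills the price (Free at zero) and license label.
# The License click broadcasts ShowItemLicense; the ellipsis toggles a help/report overlay.
# For focus navigation, it restores the last focused button and hands downward moves to an externally supplied widget, broadcasting an event.
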WidgetBlueprint /CreativeEmporium/Widgets/WBP_Creative_Emporium_Item_Price
graphs [
 EventGraph {nodes [
   "On Clicked (Button_License)"!() -> (OutputDelegate, then, Button:Common Button Base Object Reference) {is=title K2Node_ComponentBoundEvent}
   > Call Show Item License!(
    execute:Exec
    self:WBP Creative Emporium Item Price Object Reference
    ItemData:Fort Emporium Item Data Structure=Get ItemData() -> (ItemData:Fort Emporium Item Data Structure)
   ) -> (then:Exec) {is=title K2Node_CallDelegate}
   "On Clicked (Button_Ellipsis)"!() -> (OutputDelegate, then, Button:Common Button Base Object Reference) {is=title K2Node_ComponentBoundEvent}
   > n1 = Branch(Condition=Get bIsShowingAdditionalButtons() -> (bIsShowingAdditionalButtons:bool))
   > HideAdditionalButtons()
   ShowAdditionalButtons(execute=@n1.else)
   n2 = "On Focus Changed (Button_Action)"!() -> (OutputDelegate, then, HasFocus:bool) {is=title K2Node_ComponentBoundEvent}
   > UpdateLastFocusedButton!(
    execute
    self
    NewFocusedButton:WBP UIKit Button CTA Base Object Reference=Get Button_Action() -> (Button_Action:WBP UIKit Button Loud Object Reference)
    HasFocus:bool=@n2.HasFocus
   ) -> (then)
   n3 = "On Focus Changed (Button_Add)"!() -> (OutputDelegate, then, HasFocus:bool) {is=title K2Node_ComponentBoundEvent}
   > UpdateLastFocusedButton(NewFocusedButton=Get Button_Add() -> (Button_Add:WBP UIKit Button Regular Object Reference), HasFocus=@n3.HasFocus)
   n4 = "On Focus Changed (Button_License)"!() -> (OutputDelegate, then, HasFocus:bool) {is=title K2Node_ComponentBoundEvent}
   > UpdateLastFocusedButton(NewFocusedButton=Get Button_License() -> (Button_License:WBP UIKit Button Regular Object Reference), HasFocus=@n4.HasFocus)
   n5 = "On Focus Changed (Button_Ellipsis)"!() -> (OutputDelegate, then, HasFocus:bool) {is=title K2Node_ComponentBoundEvent}
   > UpdateLastFocusedButton(NewFocusedButton=Get Button_Ellipsis() -> (Button_Ellipsis:WBP UIKit Button Regular Object Reference), HasFocus=@n5.HasFocus)
   n6 = Event PreConstruct()
   > Branch(Condition=@n6.IsDesignTime)
   > SetDesignerText!(execute, self, InPrice:Text="INVTEXT(\"2000\")", InLicence:Text="INVTEXT(\"Licence: Standard\")") -> (then)
   n7 = Event UpdateWithFortItem!() -> (OutputDelegate, then, Item:Fort Item Object Reference, ParentDefinition:Item Definition Base Object Reference, bAllowInteractiveTags:bool) {is=title K2Node_Event}
   > n8 = Set Item(Item:Fort Item Object Reference=@n7.Item) -> (Output_Get:Fort Item Object Reference)
   > Is Valid(InputObject=@n8)
   > ResetText()
   > UpdateItemData()
   > PopulateItemPrice()
   > PopulateLicense()
  ]}
 HideAdditionalButtons {nodes [
   HideAdditionalButtons() {is=title K2Node_FunctionEntry}
   > Set bIsShowingAdditionalButtons(bIsShowingAdditionalButtons:bool=false) -> (Output_Get:bool)
   > SetVisibility(self=Get Overlay_AdditionalButtons() -> (Overlay_AdditionalButtons:Overlay Object Reference), InVisibility=Hidden)
   > SetFocus(self=Get Button_Ellipsis())
  ]}
 IsShowingAdditionalButtons {nodes [
   IsShowingAdditionalButtons() {is=title K2Node_FunctionEntry}
   > Return Node!(execute, IsShowingAdditionalButtons:bool=Get bIsShowingAdditionalButtons())
  ]}
 LastItemNavigationDownEvent {nodes [LastItemNavigationDownEvent() {is=title K2Node_FunctionEntry}]}
 NavigateDownFromLastButton {nodes [
   NavigateDownFromLastButton!() -> (then, Navigation:EUINavigation Enum) {is=title K2Node_FunctionEntry}
   > Call Last Item Navigation Down Event!(execute:Exec, self:WBP Creative Emporium Item Price Object Reference) -> (then:Exec) {is=title K2Node_CallDelegate}
   > Return Node!(
    execute
    ReturnValue:Widget Object Reference=Get WidgetToNavigateToOnDown() -> (WidgetToNavigateToOnDown:Common User Widget Object Reference)
   )
  ]}
 NavigateLeftFromAdditionalButtons {nodes [
   NavigateLeftFromAdditionalButtons!() -> (then, Navigation:EUINavigation Enum) {is=title K2Node_FunctionEntry}
   > HideAdditionalButtons() {is=title K2Node_CallFunction}
   > Return Node(ReturnValue=Get Button_Ellipsis())
  ]}
 OnButtonFocusChanged {nodes [
   n1 = OnButtonFocusChanged!() -> (then, HasFocus:bool=~) {is=title K2Node_FunctionEntry}
   > Branch(Condition=@n1)
  ]}
 OnFocusReceived {nodes [
   OnFocusReceived!() -> (then, MyGeometry:Geometry Structure, InFocusEvent:Focus Event Structure) {is=title K2Node_FunctionEntry}
   > n1 = Get!(execute:Exec) -> (then:Exec, else:Exec, LastFocusedButton:WBP UIKit Button CTA Base Object Reference)
   > n2 = SetFocus(self=Get LastFocusedButton() -> (LastFocusedButton:WBP UIKit Button CTA Base Object Reference))
   Return Node(execute=[@n2.then, @n3.then], ReturnValue:Event Reply Structure=Handled())
   n3 = SetFocus(execute=@n1.else, self=Get Button_Action())
  ]}
 PopulateItemPrice {nodes [
   PopulateItemPrice() {is=title K2Node_FunctionEntry}
   n1 = Break Fort Emporium Item Data!(FortEmporiumItemData:"Fort Emporium Item Data Structure (by ref)"=Get ItemData()) -> (Price:float=~) {is=title K2Node_BreakStruct}
   > "SetText (Text)"(
    self=Get Text_Price() -> (Text_Price:Common Numeric Text Block Object Reference)
    InText=Select(
     Option 0:Text="To Text (Float)"(Value=@n1, bUseGrouping=false, MinimumIntegralDigits=1, MaximumIntegralDigits=324, MinimumFractionalDigits=2, MaximumFractionalDigits=2)
     Option 1:Text="LOCTEXT(\"Free\")"
     Index="Equal (Float)"(A=@n1)
    ) -> (ReturnValue:Text)
   )
  ]}
 PopulateLicense {nodes [
   PopulateLicense() {is=title K2Node_FunctionEntry}
   n1 = Break Fort Emporium Item Data!(FortEmporiumItemData=Get ItemData()) -> (
    Id:String
    Title:String
    Namespace:String
    EntityType:String
    ThumbnailURL:String
    Price=~
    BaseCurrency:String
    SketchfabUID:String
    TagList:Array of Names
    LicenseType:EFortEmporiumItemLicense Enum
    bRequiresEntitlement:bool=~
    VersePath:String
    AssetIds:Array of Strings
    FileSize:int
    Description:String
    Seller:String
    MaterialCount:int
    PolygonCount:int
    VertexCount:int
    ChildAssetCount:int
   )
   > SetText!(
    execute
    self:Fort CTAButton Object Reference=Get Button_License()
    ButtonText:"Text (by ref)"=Format Text!(
     Format="LOCTEXT(\"{Title}: {License Type}\")"
     Title:Text=MakeLiteralText(Value="LOCTEXT(\"License\")")
     License Type:"Text (by ref)"=GetTextFromEmporiumLicense!(License:EFortEmporiumItemLicense Enum=@n1.LicenseType) -> (ReturnValue:"Text (by ref)")
    ) -> (Result)
   ) -> (then)
  ]}
 ProvideReferenceToDownNavigatedWidget {nodes [
   n1 = ProvideReferenceToDownNavigatedWidget!() -> (then, Widget:Common User Widget Object Reference) {is=title K2Node_FunctionEntry}
   > Set WidgetToNavigateToOnDown(WidgetToNavigateToOnDown:Common User Widget Object Reference=@n1) -> (Output_Get:Common User Widget Object Reference)
  ]}
 ResetText {nodes [
   ResetText() {is=title K2Node_FunctionEntry}
   n1 = GetEmptyText!() -> (ReturnValue:Text)
   > "SetText (Text)"(self=Get Text_Price(), InText=@n1)
   > SetText(self=Get Button_License(), ButtonText=@n1)
  ]}
 SetDesignerText {nodes [
   SetDesignerText!() -> (then, InPrice:Text, InLicence:Text) {is=title K2Node_FunctionEntry}
   > "SetText (Text)"(self=Get Text_Price(), InText=Get InPrice() -> (InPrice:Text))
   > SetText(self=Get Button_License(), ButtonText=Get InLicence() -> (InLicence:Text))
  ]}
 ShowAdditionalButtons {nodes [
   ShowAdditionalButtons() {is=title K2Node_FunctionEntry}
   > Set bIsShowingAdditionalButtons(bIsShowingAdditionalButtons=true)
   > SetVisibility(self=Get Overlay_AdditionalButtons(), InVisibility=SelfHitTestInvisible)
   > SetFocus(self=Get Button_HelpWithThisOrder() -> (Button_HelpWithThisOrder:WBP UIKit Button Regular Object Reference))
  ]}
 ShowAdditionalItemData {nodes [ShowAdditionalItemData!() -> (then, Item:Fort Item Object Reference) {is=title K2Node_FunctionEntry}]}
 ShowItemLicense {nodes [ShowItemLicense!() -> (then, ItemData:Fort Emporium Item Data Structure) {is=title K2Node_FunctionEntry}]}
 UpdateItemData {nodes [
   UpdateItemData() {is=title K2Node_FunctionEntry}
   > n1 = Cast To FortEmporiumItemDefinition!(
    execute
    Object=GetItemDefinition!(self:Fort Item Object Reference=Get Item() -> (Item:Fort Item Object Reference)) -> (ReturnValue:Fort Item Definition Object Reference)
   ) -> (then, CastFailed, AsFort Emporium Item Definition:Fort Emporium Item Definition Object Reference) {is=title K2Node_DynamicCast}
   > Set ItemData(ItemData:Fort Emporium Item Data Structure=GetEmporiumItemData!(self:Fort Emporium Item Definition Object Reference=@n1) -> (ReturnValue:"Fort Emporium Item Data Structure (by ref)")) -> (Output_Get:Fort Emporium Item Data Structure)
  ]}
 UpdateLastFocusedButton {nodes [
   n1 = UpdateLastFocusedButton!() -> (then, NewFocusedButton:WBP UIKit Button CTA Base Object Reference, HasFocus:bool=~) {is=title K2Node_FunctionEntry}
   > Branch(Condition=@n1.HasFocus)
   > Set LastFocusedButton(LastFocusedButton:WBP UIKit Button CTA Base Object Reference=Get NewFocusedButton() -> (NewFocusedButton:WBP UIKit Button CTA Base Object Reference)) -> (Output_Get:WBP UIKit Button CTA Base Object Reference)
  ]}
]
variables [
 bIsShowingAdditionalButtons {type=boolean}
 Item {type=FortItem, default=null}
 ItemData {
  type=FortEmporiumItemData
  default {
   id=""
   bRequiresEntitlement=false
   baseCurrency=""
   childAssetCount=0
   description=""
   entityType=""
   fileSize=0
   licenseType=Invalid
   materialCount=0
   namespace=""
   polygonCount=0
   price=0
   seller=""
   sketchfabUId=""
   thumbnailURL=""
   title=""
   versePath=""
   vertexCount=0
   assetIds []
   tagList []
  }
 }
 LastFocusedButton {type=WBP_UIKit_ButtonCTA_Base_C, default=null}
 LastItemNavigationDownEvent {}
 ShowAdditionalItemData {}
 ShowItemLicense {}
 WidgetToNavigateToOnDown {type=CommonUserWidget, default=null}
]
widgets {
 info {inheritedWidgetCount=0, parentClass=/Script/CreativeEmporiumUI.FortEmporiumItemInfo, rootWidgetClass=&GridPanel, widgetCount=13}
 widgets [
  Root {widgetClassPath=&GridPanel, bIsVariable=true, widget=Root, properties {columnFill [1], rowFill [1]}}
  HorizontalBox_Currency {
   widgetClassPath=&HorizontalBox
   parent=Root
   widget=HorizontalBox_Currency
   slot {class=GridSlot, properties {horizontalAlignment=HAlign_Center, verticalAlignment=VAlign_Top, padding=<ltrb 0 0 0 30>}}
  }
  Image_Vbucks {
   widgetClassPath=&Image
   bIsVariable=true
   parent=HorizontalBox_Currency
   widget=Image_Vbucks
   properties {brush {
     bIsDynamicallyLoaded=false
     drawAs=Image
     imageType=NoImage
     mirroring=NoMirror
     resourceName=None
     resourceObject=/Game/UI/Foundation/Textures/Icons/Currency/T-VBuck-64
     tiling=NoTile
     imageSize=<32 32>
     margin=<ltrb 0 0 0 0>
     outlineSettings {
      bUseBrushTransparency=false
      roundingType=HalfHeightRadius
      width=0
      color {colorUseRule=UseColor_Specified, specifiedColor=<rgba 0 0 0 0>}
      cornerRadii=<0 0 0 0>
     }
     tintColor {colorUseRule=UseColor_Specified, specifiedColor=<rgba 1 1 1 1>}
     uVRegion {bIsValid=false, max=<0 0>, min=<0 0>}
    }}
   slot {class=HorizontalBoxSlot, properties {horizontalAlignment=HAlign_Center, verticalAlignment=VAlign_Center}}
  }
  Text_Price {
   widgetClassPath=/Script/CommonUI.CommonNumericTextBlock
   bIsVariable=true
   parent=HorizontalBox_Currency
   widget=Text_Price
   properties {
    currentNumericValue=2000
    justification=Center
    text=0
    font {
     bForceMonospaced=false
     bMaterialIsStencil=false
     fontMaterial=null
     fontObject=&Burbank
     letterSpacing=0
     monospacedWidth=1
     size=32
     skewAmount=0
     typefaceFontName=BurbankRegular-Black
     outlineSettings {
      bApplyOutlineToDropShadows=false
      bMiteredCorners=false
      bSeparateFillAlpha=false
      outlineMaterial=null
      outlineSize=0
      outlineColor=<rgba 0 0 0 1>
     }
    }
    shadowOffset=<0 0>
   }
   slot {class=HorizontalBoxSlot, properties {horizontalAlignment=HAlign_Center, verticalAlignment=VAlign_Center, padding=<ltrb 8 0 0 0>}}
  }
  Button_Action {
   widgetClassPath=/FortUILibrary/UIKit/Buttons/WBP_UIKit_Button_Loud
   bIsVariable=true
   parent=Root
   widget=Button_Action
   properties {minWidth=399, text=Place Now}
   slot {class=GridSlot, properties {horizontalAlignment=HAlign_Center, row=1, verticalAlignment=VAlign_Top, padding=<ltrb 0 0 0 10>}}
  }
  Button_Add {
   widgetClassPath=&WBP_UIKit_Button_Regular
   bIsVariable=true
   parent=Root
   widget=Button_Add
   properties {buttonSize=Small, buttonType=Quiet, minHeight=56, minWidth=399, text=Add to Library}
   slot {class=GridSlot, properties {horizontalAlignment=HAlign_Center, row=2, verticalAlignment=VAlign_Top, padding=<ltrb 0 0 0 10>}}
  }
  Button_License {
   widgetClassPath=&WBP_UIKit_Button_Regular
   bIsVariable=true
   parent=Root
   widget=Button_License
   properties {
    buttonSize=Small
    buttonType=Quiet
    minHeight=56
    minWidth=316
    navigation=&self:WidgetTree.Button_License.WidgetNavigation_1
    text="[PH] Licence: Standard"
   }
   slot {class=GridSlot, properties {horizontalAlignment=HAlign_Left, row=3, verticalAlignment=VAlign_Top}}
  }
  Button_Ellipsis {
   widgetClassPath=&WBP_UIKit_Button_Regular
   bIsVariable=true
   parent=Root
   widget=Button_Ellipsis
   properties {
    buttonSize=Small
    buttonType=Quiet
    minHeight=56
    minWidth=60
    navigation=&self:WidgetTree.Button_Ellipsis.WidgetNavigation_0
    text=...
   }
   slot {class=GridSlot, properties {horizontalAlignment=HAlign_Right, row=3, verticalAlignment=VAlign_Top}}
  }
  Overlay_AdditionalButtons {
   widgetClassPath=&Overlay
   bIsVariable=true
   parent=Root
   widget=Overlay_AdditionalButtons
   properties {visibility=Hidden}
   slot {class=GridSlot, properties {column=1, horizontalAlignment=HAlign_Left, row=3, verticalAlignment=VAlign_Top, padding=<ltrb 33 -22 0 0>}}
  }
  Image_Background {
   widgetClassPath=&Image
   parent=Overlay_AdditionalButtons
   widget=Image_Background
   properties {brush {
     bIsDynamicallyLoaded=false
     drawAs=Image
     imageType=NoImage
     mirroring=NoMirror
     resourceName=None
     resourceObject=/CreativeEmporium/Materials/MI_Creative_Emporium_Background_Item_Description
     tiling=NoTile
     imageSize=<32 32>
     margin=<ltrb 0 0 0 0>
     outlineSettings {
      bUseBrushTransparency=false
      roundingType=HalfHeightRadius
      width=0
      color {colorUseRule=UseColor_Specified, specifiedColor=<rgba 0 0 0 0>}
      cornerRadii=<0 0 0 0>
     }
     tintColor {colorUseRule=UseColor_Specified, specifiedColor=<rgba 1 1 1 1>}
     uVRegion {bIsValid=false, max=<0 0>, min=<0 0>}
    }}
   slot {class=OverlaySlot, properties {horizontalAlignment=HAlign_Fill, verticalAlignment=VAlign_Fill}}
  }
  VerticalBox_AdditionalButtons {
   widgetClassPath=&VerticalBox
   parent=Overlay_AdditionalButtons
   widget=VerticalBox_AdditionalButtons
   slot {class=OverlaySlot, properties {padding=<ltrb 33 23 33 23>}}
  }
  Button_HelpWithThisOrder {
   widgetClassPath=&WBP_UIKit_Button_Regular
   bIsVariable=true
   parent=VerticalBox_AdditionalButtons
   widget=Button_HelpWithThisOrder
   properties {
    buttonSize=Small
    buttonType=Quiet
    minHeight=56
    minWidth=60
    navigation=&self:WidgetTree.Button_HelpWithThisOrder.WidgetNavigation_0
    text=HELP WITH THIS ORDER
   }
   slot {class=VerticalBoxSlot, properties {verticalAlignment=VAlign_Top}}
  }
  Button_ReportContent {
   widgetClassPath=&WBP_UIKit_Button_Regular
   bIsVariable=true
   parent=VerticalBox_AdditionalButtons
   widget=Button_ReportContent
   properties {
    buttonSize=Small
    buttonType=Quiet
    minHeight=56
    minWidth=60
    navigation=&self:WidgetTree.Button_ReportContent.WidgetNavigation_0
    text=REPORT CONTENT
   }
   slot {class=VerticalBoxSlot, properties {verticalAlignment=VAlign_Top, padding=<ltrb 0 19 0 0>}}
  }
 ]
}
