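# Base sub-panel widget for Creative Emporium item UI, showing a section title above a content area.
# A vertical box stacks a Common UI title text block over a padded, top-left aligned named slot for content supplied by users of the widget.
# PreConstruct copies the Title variable into the text block, replacing its placeholder text.
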WidgetBlueprint /CreativeEmporium/Widgets/WBP_Creative_Emporium_Item_SubPanel_Base
graphs [EventGraph {nodes [
   Event PreConstruct()
   > "SetText (Text)"(self=Get Text_Title() -> (Text_Title:Common Text Object Reference), InText=Get Title() -> (Title:Text))
  ]}]
variables [Title {type=string, default="[PH] TITLE"}]
widgets {
 info {inheritedWidgetCount=0, parentClass=&CommonUserWidget, rootWidgetClass=&VerticalBox, widgetCount=3}
 widgets [
  VerticalBox_Root {widgetClassPath=&VerticalBox, widget=VerticalBox_Root}
  Text_Title {
   widgetClassPath=&CommonTextBlock
   bIsVariable=true
   parent=VerticalBox_Root
   widget=Text_Title
   properties {
    style=/FortUILibrary/Styles/Typography/_TextStyle_SubH4._TextStyle_SubH4_C
    text="[PH] SECTION TITLE"
    font {
     bForceMonospaced=false
     bMaterialIsStencil=false
     fontMaterial=null
     fontObject=&BurbankSmall
     letterSpacing=0
     monospacedWidth=1
     size=24
     skewAmount=0
     typefaceFontName=Black
     outlineSettings {
      bApplyOutlineToDropShadows=false
      bMiteredCorners=false
      bSeparateFillAlpha=false
      outlineMaterial=null
      outlineSize=0
      outlineColor=<rgba 0 0 0 1>
     }
    }
    shadowOffset=<0 0>
    strikeBrush {
     bIsDynamicallyLoaded=false
     drawAs=NoDrawType
     imageType=NoImage
     mirroring=NoMirror
     resourceName=None
     resourceObject=null
     tiling=NoTile
     imageSize=<32 32>
     margin=<ltrb 0 0 0 0>
     outlineSettings {
      bUseBrushTransparency=false
      roundingType=HalfHeightRadius
      width=0
      color {colorUseRule=UseColor_Specified, specifiedColor=<rgba 0 0 0 0>}
      cornerRadii=<0 0 0 0>
     }
     tintColor {colorUseRule=UseColor_Specified, specifiedColor=<rgba 1 1 1 1>}
     uVRegion {bIsValid=false, max=<0 0>, min=<0 0>}
    }
   }
   slot {class=VerticalBoxSlot}
  }
  NamedSlot_Content {
   widgetClassPath=&NamedSlot
   parent=VerticalBox_Root
   widget=NamedSlot_Content
   slot {class=VerticalBoxSlot, properties {horizontalAlignment=HAlign_Left, verticalAlignment=VAlign_Top, padding=<ltrb 0 25 0 0>}}
  }
 ]
}
