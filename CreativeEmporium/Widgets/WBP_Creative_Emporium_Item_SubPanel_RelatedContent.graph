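# Creative Emporium item sub-panel widget that provides the related-content section as a horizontal tile view.
# It derives from the Emporium item sub-panel base and holds a single AthenaCreativeItemTileView, with the layout frame left to the parent.
# The tile view builds square, evenly spaced entries from the RelatedContentPreview widget.
# The event graph is empty, so this asset adds no Blueprint logic of its own.
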WidgetBlueprint /CreativeEmporium/Widgets/WBP_Creative_Emporium_Item_SubPanel_RelatedContent
graphs [EventGraph {links [], nodes []}]
variables []
widgets {
 info {
  inheritedWidgetCount=0
  namedSlotCount=1
  parentClass=&/WBP_Creative_Emporium_Item_SubPanel_Base.WBP_Creative_Emporium_Item_SubPanel_Base_C
  rootWidgetClass=null
  widgetCount=1
 }
 widgets [AthenaCreativeItemTileView_RelatedContent {
   widgetClassPath=/Script/FortniteUI.AthenaCreativeItemTileView
   widget=AthenaCreativeItemTileView_RelatedContent
   properties {
    entryHeight=408
    entryWidgetClass=&/WBP_Creative_Emporium_Item_RelatedContentPreview.WBP_Creative_Emporium_Item_RelatedContentPreview_C
    entryWidth=408
    horizontalEntrySpacing=24
    orientation=Orient_Horizontal
    verticalEntrySpacing=24
   }
  }]
}
