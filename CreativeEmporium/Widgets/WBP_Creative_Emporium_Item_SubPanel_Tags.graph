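# Creative Emporium item sub-panel widget that lays out tags, derived from the shared WBP_Creative_Emporium_Item_SubPanel_Base.
# It has no root widget and adds a single DynamicEntryBox, EntryBox_ActivityTags, alongside one named slot.
# The box generates ActivityBrowserTag entries, reused from the DiscoveryShared UI, with custom spacing and designer preview entries.
# The event graph is empty, so this asset contains no logic to populate the tags.
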WidgetBlueprint /CreativeEmporium/Widgets/WBP_Creative_Emporium_Item_SubPanel_Tags
graphs [EventGraph {links [], nodes []}]
variables []
widgets {
 info {
  inheritedWidgetCount=0
  namedSlotCount=1
  parentClass=&/WBP_Creative_Emporium_Item_SubPanel_Base.WBP_Creative_Emporium_Item_SubPanel_Base_C
  rootWidgetClass=null
  widgetCount=1
 }
 widgets [EntryBox_ActivityTags {
   widgetClassPath=&DynamicEntryBox
   widget=EntryBox_ActivityTags
   properties {entryWidgetClass=/Game/UI/DiscoveryShared/ActivityBrowserTag.ActivityBrowserTag_C, numDesignerPreviewEntries=7, entrySpacing=<-2 8>}
  }]
}
