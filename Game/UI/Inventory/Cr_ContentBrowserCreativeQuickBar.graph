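# Creative content browser quick bar widget: a row of eight slot buttons, each bound to its QuickbarSlot input action.
# Construct refills EquipButtonGroup with the buttons and binds their drop and tap events.
# Clicks, taps and drops are relayed through the OnQuickbarSlotActivated and OnQuickBarDrop dispatchers.
# Clicks and drops also call Pulse on the button, converting 1-based slot IDs to array indices; taps do not pulse.
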WidgetBlueprint /Game/UI/Inventory/Cr_ContentBrowserCreativeQuickBar
graphs [
 EventGraph {nodes [
   "On Button Clicked (QuickBarButton_Slot1)"!() -> (OutputDelegate, then, Button:Common Button Legacy Object Reference) {is=title K2Node_ComponentBoundEvent}
   > Call On Quickbar Slot Activated!(execute:Exec, self:Cr Content Browser Creative Quick Bar Object Reference, SlotId:int=1) -> (then:Exec) {is=title K2Node_CallDelegate}
   > Pulse!(execute, self, SlotId:int=0) -> (then)
   "On Button Clicked (QuickBarButton_Slot2)"!() -> (OutputDelegate, then, Button:Common Button Legacy Object Reference) {is=title K2Node_ComponentBoundEvent}
   > Call On Quickbar Slot Activated(SlotId=2)
   > Pulse(SlotId=1)
   "On Button Clicked (QuickBarButton_Slot3)"!() -> (OutputDelegate, then, Button:Common Button Legacy Object Reference) {is=title K2Node_ComponentBoundEvent}
   > Call On Quickbar Slot Activated(SlotId=3)
   > Pulse(SlotId=2)
   "On Button Clicked (QuickBarButton_Slot4)"!() -> (OutputDelegate, then, Button:Common Button Legacy Object Reference) {is=title K2Node_ComponentBoundEvent}
   > Call On Quickbar Slot Activated(SlotId=4)
   > Pulse(SlotId=3)
   "On Button Clicked (QuickBarButton_Slot5)"!() -> (OutputDelegate, then, Button:Common Button Legacy Object Reference) {is=title K2Node_ComponentBoundEvent}
   > Call On Quickbar Slot Activated(SlotId=5)
   > Pulse(SlotId=4)
   "On Button Clicked (QuickBarButton_Slot6)"!() -> (OutputDelegate, then, Button:Common Button Legacy Object Reference) {is=title K2Node_ComponentBoundEvent}
   > Call On Quickbar Slot Activated(SlotId=6)
   > Pulse(SlotId=5)
   "On Button Clicked (QuickBarButton_Slot7)"!() -> (OutputDelegate, then, Button:Common Button Legacy Object Reference) {is=title K2Node_ComponentBoundEvent}
   > Call On Quickbar Slot Activated(SlotId=7)
   > Pulse(SlotId=6)
   "On Button Clicked (QuickBarButton_Slot8)"!() -> (OutputDelegate, then, Button:Common Button Legacy Object Reference) {is=title K2Node_ComponentBoundEvent}
   > Call On Quickbar Slot Activated(SlotId=8)
   > Pulse(SlotId=7)
   n1 = OnDropped!() -> (OutputDelegate, then, ItemSlot:int, QuickbarType:EFortQuickBars Enum=Primary, Item:Fort Item Object Reference) {is=title K2Node_CustomEvent}
   > Call On Quick Bar Drop!(
    execute:Exec
    self:Cr Content Browser Creative Quick Bar Object Reference
    SlotID:int=@n1.ItemSlot
    QuickbarType:EFortQuickBars Enum=@n1.QuickbarType
    Item:Fort Item Object Reference=@n1.Item
   ) -> (then:Exec) {is=title K2Node_CallDelegate}
   > Pulse(SlotId=int - int(A=@n1.ItemSlot))
   n2 = OnTapped!() -> (OutputDelegate, then, NewParam:int) {is=title K2Node_CustomEvent}
   > Call On Quickbar Slot Activated(SlotId=@n2.NewParam)
   Event Construct()
   > RemoveAll(self:Common Widget Group Base Object Reference=Get EquipButtonGroup() -> (EquipButtonGroup:Common Button Group Base Object Reference))
   > n3 = For Each Loop(Array:Array of Cr Content Browser Quick Bar Button Object References=Make Array!(
     "[0]":Cr Content Browser Quick Bar Button Object Reference=Get QuickBarButton_Slot1() -> (QuickBarButton_Slot1:Cr Content Browser Quick Bar Button Object Reference)
     "[1]":Cr Content Browser Quick Bar Button Object Reference=Get QuickBarButton_Slot2() -> (QuickBarButton_Slot2:Cr Content Browser Quick Bar Button Object Reference)
     "[2]":Cr Content Browser Quick Bar Button Object Reference=Get QuickBarButton_Slot3() -> (QuickBarButton_Slot3:Cr Content Browser Quick Bar Button Object Reference)
     "[3]":Cr Content Browser Quick Bar Button Object Reference=Get QuickBarButton_Slot4() -> (QuickBarButton_Slot4:Cr Content Browser Quick Bar Button Object Reference)
     "[4]":Cr Content Browser Quick Bar Button Object Reference=Get QuickBarButton_Slot5() -> (QuickBarButton_Slot5:Cr Content Browser Quick Bar Button Object Reference)
     "[5]":Cr Content Browser Quick Bar Button Object Reference=Get QuickBarButton_Slot6() -> (QuickBarButton_Slot6:Cr Content Browser Quick Bar Button Object Reference)
     "[6]":Cr Content Browser Quick Bar Button Object Reference=Get QuickBarButton_Slot7() -> (QuickBarButton_Slot7:Cr Content Browser Quick Bar Button Object Reference)
     "[7]":Cr Content Browser Quick Bar Button Object Reference=Get QuickBarButton_Slot8() -> (QuickBarButton_Slot8:Cr Content Browser Quick Bar Button Object Reference)
    ) -> (Array:Array of Cr Content Browser Quick Bar Button Object References)) -> (Array Element:Cr Content Browser Quick Bar Button Object Reference)
   > AddWidget!(execute, self:Common Widget Group Base Object Reference=Get EquipButtonGroup(), InWidget:Widget Object Reference=@n3) -> (then)
   > Bind Event to On Item Dropped(self:Cr Content Browser Quick Bar Button Object Reference=@n3, Delegate=@n1) {is=title K2Node_AddDelegate}
   > Bind Event to On Item Tapped(self:Cr Content Browser Quick Bar Button Object Reference=@n3, Delegate=@n2) {is=title K2Node_AddDelegate}
  ]}
 OnQuickBarDrop {nodes [OnQuickBarDrop!() -> (then, SlotID:int, QuickbarType:EFortQuickBars Enum=Primary, Item:Fort Item Object Reference) {is=title K2Node_FunctionEntry}]}
 OnQuickbarSlotActivated {nodes [OnQuickbarSlotActivated!() -> (then, SlotId:int) {is=title K2Node_FunctionEntry}]}
 Pulse {nodes [
   n1 = Pulse!() -> (then, SlotId:int) {is=title K2Node_FunctionEntry}
   n2 = Get QuickBarButton_Slot1()
   n3 = Get QuickBarButton_Slot2()
   n4 = Get QuickBarButton_Slot3()
   n5 = Get QuickBarButton_Slot4()
   n6 = Get QuickBarButton_Slot5()
   n7 = Get QuickBarButton_Slot6()
   n8 = Get QuickBarButton_Slot7()
   n9 = Get QuickBarButton_Slot8()
   > Event_Pulse(self:Cr Content Browser Quick Bar Button Object Reference="Get (a copy)"(
     Array:Array of Cr Content Browser Quick Bar Button Object References=Make Array("[0]"=@n2, "[1]"=@n3, "[2]"=@n4, "[3]"=@n5, "[4]"=@n6, "[5]"=@n7, "[6]"=@n8, "[7]"=@n9)
     Dimension 1=@n1
    ) -> (Output:Cr Content Browser Quick Bar Button Object Reference))
  ]}
]
variables [
 OnQuickBarDrop {}
 OnQuickbarSlotActivated {}
]
widgets {
 info {inheritedWidgetCount=0, parentClass=/Script/FortniteUI.AthenaQuickbarEditorBase, rootWidgetClass=&Overlay, widgetCount=10}
 widgets [
  Overlay_61 {widgetClassPath=&Overlay, widget=Overlay_61}
  EquipmentSlots {
   widgetClassPath=&HorizontalBox
   parent=Overlay_61
   widget=EquipmentSlots
   properties {navigation=&self:WidgetTree.EquipmentSlots.WidgetNavigation_0}
   slot {class=OverlaySlot}
  }
  QuickBarButton_Slot1 {
   widgetClassPath=&/Cr_ContentBrowserQuickBarButton
   bIsVariable=true
   parent=EquipmentSlots
   widget=QuickBarButton_Slot1
   properties {
    creativeSlotIndex=1
    onButtonBaseClicked=(&self:WidgetTree.QuickBarButton_Slot1.HandleOnButtonClicked)
    onButtonBaseDoubleClicked=(&self:WidgetTree.QuickBarButton_Slot1.HandleOnButtonDoubleClicked)
    onButtonBaseHovered=(&self:WidgetTree.QuickBarButton_Slot1.HandleOnButtonHovered)
    onButtonBaseUnhovered=(&self:WidgetTree.QuickBarButton_Slot1.HandleOnButtonUnhovered)
    onSelectedChangedBase=(&self:WidgetTree.QuickBarButton_Slot1.HandleOnSelectedChanged)
    triggeringInputAction {dataTable=/FortUILibrary/Input/CreativeInputActionDataTable, rowName=QuickbarSlot1}
   }
   slot {class=HorizontalBoxSlot}
  }
  QuickBarButton_Slot2 {
   widgetClassPath=&Cr_ContentBrowserQuickBarButton
   bIsVariable=true
   parent=EquipmentSlots
   widget=QuickBarButton_Slot2
   properties {
    creativeSlotIndex=2
    onButtonBaseClicked=(&self:WidgetTree.QuickBarButton_Slot2.HandleOnButtonClicked)
    onButtonBaseDoubleClicked=(&self:WidgetTree.QuickBarButton_Slot2.HandleOnButtonDoubleClicked)
    onButtonBaseHovered=(&self:WidgetTree.QuickBarButton_Slot2.HandleOnButtonHovered)
    onButtonBaseUnhovered=(&self:WidgetTree.QuickBarButton_Slot2.HandleOnButtonUnhovered)
    onSelectedChangedBase=(&self:WidgetTree.QuickBarButton_Slot2.HandleOnSelectedChanged)
    triggeringInputAction {dataTable=&CreativeInputActionDataTable, rowName=QuickbarSlot2}
   }
   slot {class=HorizontalBoxSlot}
  }
  QuickBarButton_Slot3 {
   widgetClassPath=&Cr_ContentBrowserQuickBarButton
   bIsVariable=true
   parent=EquipmentSlots
   widget=QuickBarButton_Slot3
   properties {
    creativeSlotIndex=3
    onButtonBaseClicked=(&self:WidgetTree.QuickBarButton_Slot3.HandleOnButtonClicked)
    onButtonBaseDoubleClicked=(&self:WidgetTree.QuickBarButton_Slot3.HandleOnButtonDoubleClicked)
    onButtonBaseHovered=(&self:WidgetTree.QuickBarButton_Slot3.HandleOnButtonHovered)
    onButtonBaseUnhovered=(&self:WidgetTree.QuickBarButton_Slot3.HandleOnButtonUnhovered)
    onSelectedChangedBase=(&self:WidgetTree.QuickBarButton_Slot3.HandleOnSelectedChanged)
    triggeringInputAction {dataTable=&CreativeInputActionDataTable, rowName=QuickbarSlot3}
   }
   slot {class=HorizontalBoxSlot}
  }
  QuickBarButton_Slot4 {
   widgetClassPath=&Cr_ContentBrowserQuickBarButton
   bIsVariable=true
   parent=EquipmentSlots
   widget=QuickBarButton_Slot4
   properties {
    creativeSlotIndex=4
    onButtonBaseClicked=(&self:WidgetTree.QuickBarButton_Slot4.HandleOnButtonClicked)
    onButtonBaseDoubleClicked=(&self:WidgetTree.QuickBarButton_Slot4.HandleOnButtonDoubleClicked)
    onButtonBaseHovered=(&self:WidgetTree.QuickBarButton_Slot4.HandleOnButtonHovered)
    onButtonBaseUnhovered=(&self:WidgetTree.QuickBarButton_Slot4.HandleOnButtonUnhovered)
    onSelectedChangedBase=(&self:WidgetTree.QuickBarButton_Slot4.HandleOnSelectedChanged)
    triggeringInputAction {dataTable=&CreativeInputActionDataTable, rowName=QuickbarSlot4}
   }
   slot {class=HorizontalBoxSlot}
  }
  QuickBarButton_Slot5 {
   widgetClassPath=&Cr_ContentBrowserQuickBarButton
   bIsVariable=true
   parent=EquipmentSlots
   widget=QuickBarButton_Slot5
   properties {
    creativeSlotIndex=5
    onButtonBaseClicked=(&self:WidgetTree.QuickBarButton_Slot5.HandleOnButtonClicked)
    onButtonBaseDoubleClicked=(&self:WidgetTree.QuickBarButton_Slot5.HandleOnButtonDoubleClicked)
    onButtonBaseHovered=(&self:WidgetTree.QuickBarButton_Slot5.HandleOnButtonHovered)
    onButtonBaseUnhovered=(&self:WidgetTree.QuickBarButton_Slot5.HandleOnButtonUnhovered)
    onSelectedChangedBase=(&self:WidgetTree.QuickBarButton_Slot5.HandleOnSelectedChanged)
    triggeringInputAction {dataTable=&CreativeInputActionDataTable, rowName=QuickbarSlot5}
   }
   slot {class=HorizontalBoxSlot}
  }
  QuickBarButton_Slot6 {
   widgetClassPath=&Cr_ContentBrowserQuickBarButton
   bIsVariable=true
   parent=EquipmentSlots
   widget=QuickBarButton_Slot6
   properties {
    creativeSlotIndex=6
    onButtonBaseClicked=(&self:WidgetTree.QuickBarButton_Slot6.HandleOnButtonClicked)
    onButtonBaseDoubleClicked=(&self:WidgetTree.QuickBarButton_Slot6.HandleOnButtonDoubleClicked)
    onButtonBaseHovered=(&self:WidgetTree.QuickBarButton_Slot6.HandleOnButtonHovered)
    onButtonBaseUnhovered=(&self:WidgetTree.QuickBarButton_Slot6.HandleOnButtonUnhovered)
    onSelectedChangedBase=(&self:WidgetTree.QuickBarButton_Slot6.HandleOnSelectedChanged)
    triggeringInputAction {dataTable=&CreativeInputActionDataTable, rowName=QuickbarSlot6}
   }
   slot {class=HorizontalBoxSlot}
  }
  QuickBarButton_Slot7 {
   widgetClassPath=&Cr_ContentBrowserQuickBarButton
   bIsVariable=true
   parent=EquipmentSlots
   widget=QuickBarButton_Slot7
   properties {
    creativeSlotIndex=7
    onButtonBaseClicked=(&self:WidgetTree.QuickBarButton_Slot7.HandleOnButtonClicked)
    onButtonBaseDoubleClicked=(&self:WidgetTree.QuickBarButton_Slot7.HandleOnButtonDoubleClicked)
    onButtonBaseHovered=(&self:WidgetTree.QuickBarButton_Slot7.HandleOnButtonHovered)
    onButtonBaseUnhovered=(&self:WidgetTree.QuickBarButton_Slot7.HandleOnButtonUnhovered)
    onSelectedChangedBase=(&self:WidgetTree.QuickBarButton_Slot7.HandleOnSelectedChanged)
    triggeringInputAction {dataTable=&CreativeInputActionDataTable, rowName=QuickbarSlot7}
   }
   slot {class=HorizontalBoxSlot}
  }
  QuickBarButton_Slot8 {
   widgetClassPath=&Cr_ContentBrowserQuickBarButton
   bIsVariable=true
   parent=EquipmentSlots
   widget=QuickBarButton_Slot8
   properties {
    creativeSlotIndex=8
    onButtonBaseClicked=(&self:WidgetTree.QuickBarButton_Slot8.HandleOnButtonClicked)
    onButtonBaseDoubleClicked=(&self:WidgetTree.QuickBarButton_Slot8.HandleOnButtonDoubleClicked)
    onButtonBaseHovered=(&self:WidgetTree.QuickBarButton_Slot8.HandleOnButtonHovered)
    onButtonBaseUnhovered=(&self:WidgetTree.QuickBarButton_Slot8.HandleOnButtonUnhovered)
    onSelectedChangedBase=(&self:WidgetTree.QuickBarButton_Slot8.HandleOnSelectedChanged)
    triggeringInputAction {dataTable=&CreativeInputActionDataTable, rowName=QuickbarSlot8}
   }
   slot {class=HorizontalBoxSlot}
  }
 ]
}
