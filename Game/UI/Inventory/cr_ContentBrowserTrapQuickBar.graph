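# Creative content browser widget, derived from AthenaQuickbarEditorBase, that shows a single trap quickbar slot on the Secondary quickbar, tied to the QuickbarTrap input action.
# On construct it clears EquipButtonGroup, registers the slot button in it and binds the button's item-dropped event.
# Clicking the slot broadcasts OnQuickbarTrapSlotActivated.
# A drop broadcasts OnQuickbarDropped with only the item, leaving the equip logic to listeners.
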WidgetBlueprint /Game/UI/Inventory/cr_ContentBrowserTrapQuickBar
graphs [
 EventGraph {nodes [
   "On Button Clicked (EquipButtonSlot1)"!() -> (OutputDelegate, then, Button:Common Button Legacy Object Reference) {is=title K2Node_ComponentBoundEvent}
   > Call On Quickbar Trap Slot Activated!(execute:Exec, self:Cr Content Browser Trap Quick Bar Object Reference) -> (then:Exec) {is=title K2Node_CallDelegate}
   n1 = OnDropOnTrapQuickbar!() -> (OutputDelegate, then, ItemSlot:int, QuickbarType:EFortQuickBars Enum=Primary, Item:Fort Item Object Reference) {is=title K2Node_CustomEvent}
   > Call On Quickbar Dropped!(execute:Exec, self:Cr Content Browser Trap Quick Bar Object Reference, Item:Fort Item Object Reference=@n1.Item) -> (then:Exec) {is=title K2Node_CallDelegate}
   Event Construct()
   n2 = Get EquipButtonGroup() -> (EquipButtonGroup:Common Button Group Base Object Reference)
   > RemoveAll(self:Common Widget Group Base Object Reference=@n2)
   n3 = Get EquipButtonSlot1() -> (EquipButtonSlot1:Cr Content Browser Quick Bar Button Object Reference)
   > AddWidget!(execute, self:Common Widget Group Base Object Reference=@n2, InWidget:Widget Object Reference=@n3) -> (then)
   > Bind Event to On Item Dropped(self:Cr Content Browser Quick Bar Button Object Reference=@n3, Delegate=@n1) {is=title K2Node_AddDelegate}
  ]}
 OnQuickbarDropped {nodes [OnQuickbarDropped!() -> (then, Item:Fort Item Object Reference) {is=title K2Node_FunctionEntry}]}
 OnQuickbarTrapSlotActivated {nodes [OnQuickbarTrapSlotActivated() {is=title K2Node_FunctionEntry}]}
]
variables [
 OnQuickbarDropped {}
 OnQuickbarTrapSlotActivated {}
]
widgets {
 info {inheritedWidgetCount=0, parentClass=/Script/FortniteUI.AthenaQuickbarEditorBase, rootWidgetClass=&HorizontalBox, widgetCount=2}
 widgets [
  EquipmentSlots {
   widgetClassPath=&HorizontalBox
   widget=EquipmentSlots
   properties {navigation=&self:WidgetTree.EquipmentSlots.WidgetNavigation_0}
  }
  EquipButtonSlot1 {
   widgetClassPath=&/Cr_ContentBrowserQuickBarButton
   bIsVariable=true
   parent=EquipmentSlots
   widget=EquipButtonSlot1
   properties {
    creativeSlotIndex=4
    onButtonBaseClicked=(&self:WidgetTree.EquipButtonSlot1.HandleOnButtonClicked)
    onButtonBaseDoubleClicked=(&self:WidgetTree.EquipButtonSlot1.HandleOnButtonDoubleClicked)
    onButtonBaseHovered=(&self:WidgetTree.EquipButtonSlot1.HandleOnButtonHovered)
    onButtonBaseUnhovered=(&self:WidgetTree.EquipButtonSlot1.HandleOnButtonUnhovered)
    onSelectedChangedBase=(&self:WidgetTree.EquipButtonSlot1.HandleOnSelectedChanged)
    quickbarType=Secondary
    triggeringInputAction {dataTable=/FortUILibrary/Input/CreativeInputActionDataTable, rowName=QuickbarTrap}
   }
   slot {class=HorizontalBoxSlot}
  }
 ]
}
